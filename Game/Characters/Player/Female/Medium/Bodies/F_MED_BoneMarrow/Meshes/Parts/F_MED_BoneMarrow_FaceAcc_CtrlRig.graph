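# Control Rig for the F_MED_BoneMarrow face accessory part that makes the back hair bones follow the upper body.
# Weighted multi-bone parent constraints tie the middle, left and right back chains to spine, clavicle and arm bones.
# The parachuting_alpha variable eases off side-chain constraints progressively toward the tips, releasing them fully at 1, while the middle stays fully constrained.
# Spherical pose readers on each clavicle then weight corrective transform offsets on the matching side's hair.
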ControlRigBlueprint /Game/Characters/Player/Female/Medium/Bodies/F_MED_BoneMarrow/Meshes/Parts/F_MED_BoneMarrow_FaceAcc_CtrlRig
graphs [
 RigVMFunctionLibrary {links [], nodes []}
 RigVMModel {nodes [
   RigUnit_BeginExecution = &RigUnit_BeginExecution()
   SequenceAggregate = RigVMAggregateNode(ExecuteContext=@RigUnit_BeginExecution.ExecutePin)
   &RigUnit_ParentConstraint(
    AdvancedSettings=(RotationOrderForFilter=XZY)
    Child=(Type=Bone,Name="dyn_hair_back_02_m")
    Parents="((Item=(Type=Bone,Name=\"upperarm_r\"),Weight=0.0002),(Item=(Type=Bone,Name=\"upperarm_l\"),Weight=0.0002),(Item=(Type=Bone,Name=\"deform_upperarm_up_l\"),Weight=0.0005),(Item=(Type=Bone,Name=\"deform_upperarm_up_r\"),Weight=0.0006),(Item=(Type=Bone,Name=\"spine_02 ...(truncated)"
    Weight=1
    ExecutePin=@SequenceAggregate.A
   )
   > &RigUnit_ParentConstraint(
    AdvancedSettings=(RotationOrderForFilter=XZY)
    Child=(Type=Bone,Name="dyn_hair_back_03_m")
    Parents="((Item=(Type=Bone,Name=\"spine_01\"),Weight=0.0005),(Item=(Type=Bone,Name=\"spine_02\"),Weight=0.0017),(Item=(Type=Bone,Name=\"spine_03\"),Weight=0.0089),(Item=(Type=Bone,Name=\"spine_04\"),Weight=0.0122),(Item=(Type=Bone,Name=\"clavicle_r\"),Weight=0.1847),(Item=(T ...(truncated)"
    Weight=1
   )
   > &RigUnit_ParentConstraint(
    AdvancedSettings=(RotationOrderForFilter=XZY)
    Child=(Type=Bone,Name="dyn_hair_back_04_m")
    Parents="((Item=(Type=Bone,Name=\"spine_01\"),Weight=0.0008),(Item=(Type=Bone,Name=\"spine_02\"),Weight=0.0019),(Item=(Type=Bone,Name=\"spine_03\"),Weight=0.0098),(Item=(Type=Bone,Name=\"spine_04\"),Weight=0.0129),(Item=(Type=Bone,Name=\"clavicle_r\"),Weight=0.1819),(Item=(T ...(truncated)"
    Weight=1
   )
   > &RigUnit_ParentConstraint(
    AdvancedSettings=(RotationOrderForFilter=XZY)
    Child=(Type=Bone,Name="dyn_hair_back_05_m")
    Parents="((Item=(Type=Bone,Name=\"upperarm_l\"),Weight=0.0004),(Item=(Type=Bone,Name=\"spine_01\"),Weight=0.0089),(Item=(Type=Bone,Name=\"spine_02\"),Weight=0.0139),(Item=(Type=Bone,Name=\"spine_03\"),Weight=0.0712),(Item=(Type=Bone,Name=\"clavicle_r\"),Weight=0.0845),(Item= ...(truncated)"
    Weight=1
   )
   n1 = RigVMVariableNode() {variable=parachuting_alpha}
   Interpolate = /Script/RigVM.RigVMFunction_MathFloatLerp(A=1, B=0.5, T=@n1.Value) {is=struct RigVMUnitNode, template="Interpolate::Execute(in A,in B,in T,out Result)"}
   &RigUnit_ParentConstraint(
    AdvancedSettings=(RotationOrderForFilter=XZY)
    Child=(Type=Bone,Name="dyn_hair_back_02_l")
    Parents="((Item=(Type=Bone,Name=\"upperarm_twist_01_l\"),Weight=0.0002),(Item=(Type=Bone,Name=\"spine_02\"),Weight=0.0002),(Item=(Type=Bone,Name=\"spine_03\"),Weight=0.0014),(Item=(Type=Bone,Name=\"spine_04\"),Weight=0.0026),(Item=(Type=Bone,Name=\"upperarm_l\"),Weight=0.005 ...(truncated)"
    Weight=@Interpolate.Result
    ExecutePin=@SequenceAggregate.B
   )
   n2 = &RigVMFunction_MathFloatLerp(A=1, B=0.25, T=@n1.Value) {template="Interpolate::Execute(in A,in B,in T,out Result)"}
   > &RigUnit_ParentConstraint(
    AdvancedSettings=(RotationOrderForFilter=XZY)
    Child=(Type=Bone,Name="dyn_hair_back_03_l")
    Parents="((Item=(Type=Bone,Name=\"upperarm_twist_01_l\"),Weight=0.0006),(Item=(Type=Bone,Name=\"clavicle_r\"),Weight=0.0011),(Item=(Type=Bone,Name=\"spine_03\"),Weight=0.0012),(Item=(Type=Bone,Name=\"spine_04\"),Weight=0.0024),(Item=(Type=Bone,Name=\"upperarm_l\"),Weight=0.0 ...(truncated)"
    Weight=@n2.Result
   )
   n3 = &RigVMFunction_MathFloatLerp(A=1, B=0, T=@n1.Value) {template="Interpolate::Execute(in A,in B,in T,out Result)"}
   > &RigUnit_ParentConstraint(
    AdvancedSettings=(RotationOrderForFilter=XZY)
    Child=(Type=Bone,Name="dyn_hair_back_04_l")
    Parents="((Item=(Type=Bone,Name=\"upperarm_twist_01_l\"),Weight=0.0005),(Item=(Type=Bone,Name=\"spine_02\"),Weight=0.0014),(Item=(Type=Bone,Name=\"spine_03\"),Weight=0.0062),(Item=(Type=Bone,Name=\"spine_04\"),Weight=0.0132),(Item=(Type=Bone,Name=\"upperarm_l\"),Weight=0.042 ...(truncated)"
    Weight=@n3.Result
   )
   > &RigUnit_ParentConstraint(
    AdvancedSettings=(RotationOrderForFilter=XZY)
    Child=(Type=Bone,Name="dyn_hair_back_05_l")
    Parents="((Item=(Type=Bone,Name=\"clavicle_r\"),Weight=0.0003),(Item=(Type=Bone,Name=\"pelvis\"),Weight=0.0004),(Item=(Type=Bone,Name=\"spine_01\"),Weight=0.0011),(Item=(Type=Bone,Name=\"spine_02\"),Weight=0.0022),(Item=(Type=Bone,Name=\"upperarm_twist_01_l\"),Weight=0.0148) ...(truncated)"
    Weight=@n3.Result
   )
   &RigUnit_ParentConstraint(
    AdvancedSettings=(RotationOrderForFilter=XZY)
    Child=(Type=Bone,Name="dyn_hair_back_02_r")
    Parents="((Item=(Type=Bone,Name=\"spine_02\"),Weight=0.0002),(Item=(Type=Bone,Name=\"spine_03\"),Weight=0.0011),(Item=(Type=Bone,Name=\"spine_04\"),Weight=0.002),(Item=(Type=Bone,Name=\"upperarm_r\"),Weight=0.0045),(Item=(Type=Bone,Name=\"clavicle_l\"),Weight=0.0123),(Item=( ...(truncated)"
    Weight=@Interpolate.Result
    ExecutePin=@SequenceAggregate.C
   )
   > &RigUnit_ParentConstraint(
    AdvancedSettings=(RotationOrderForFilter=XZY)
    Child=(Type=Bone,Name="dyn_hair_back_03_r")
    Parents="((Item=(Type=Bone,Name=\"spine_03\"),Weight=0.0003),(Item=(Type=Bone,Name=\"upperarm_twist_01_r\"),Weight=0.0004),(Item=(Type=Bone,Name=\"spine_04\"),Weight=0.0009),(Item=(Type=Bone,Name=\"upperarm_r\"),Weight=0.0176),(Item=(Type=Bone,Name=\"deform_upperarm_up_r\"), ...(truncated)"
    Weight=@n2.Result
   )
   > &RigUnit_ParentConstraint(
    AdvancedSettings=(RotationOrderForFilter=XZY)
    Child=(Type=Bone,Name="dyn_hair_back_04_r")
    Parents="((Item=(Type=Bone,Name=\"spine_02\"),Weight=0.0012),(Item=(Type=Bone,Name=\"spine_03\"),Weight=0.0059),(Item=(Type=Bone,Name=\"spine_04\"),Weight=0.0122),(Item=(Type=Bone,Name=\"upperarm_r\"),Weight=0.0383),(Item=(Type=Bone,Name=\"deform_upperarm_up_r\"),Weight=0.10 ...(truncated)"
    Weight=@n3.Result
   )
   > &RigUnit_ParentConstraint(
    AdvancedSettings=(RotationOrderForFilter=XZY)
    Child=(Type=Bone,Name="dyn_hair_back_05_r")
    Parents="((Item=(Type=Bone,Name=\"spine_01\"),Weight=0.0019),(Item=(Type=Bone,Name=\"spine_02\"),Weight=0.0028),(Item=(Type=Bone,Name=\"upperarm_twist_01_r\"),Weight=0.0117),(Item=(Type=Bone,Name=\"deform_upperarm_up_r\"),Weight=0.032),(Item=(Type=Bone,Name=\"spine_03\"),Wei ...(truncated)"
    Weight=@n3.Result
   )
   n4 = &RigUnit_SphericalPoseReader(
    ActiveRegionScaleFactors=(PositiveWidth=1,NegativeWidth=2,PositiveHeight=2,NegativeHeight=2)
    ActiveRegionSize=0.1
    Debug=(bDrawDebug=true,bDraw2D=false,bDrawLocalAxes=false,DebugScale=25,DebugSegments=20,DebugThickness=0.25)
    DriverAxis=(X=1,Y=0,Z=0)
    DriverItem=(Type=Bone,Name="clavicle_l")
    FalloffRegionScaleFactors=(PositiveWidth=1,NegativeWidth=0,PositiveHeight=0,NegativeHeight=0)
    FalloffSize=0.2
    OptionalParentItem=(Type=Bone)
    RotationOffset=(X=0,Y=-50,Z=0)
    ExecutePin=@SequenceAggregate.D
   )
   > &RigUnit_ModifyTransforms(
    ItemToModify="((Item=(Type=Bone,Name=\"dyn_hair_back_02_l\"),Transform=(Rotation=(X=0,Y=0.2588,Z=0,W=0.9659),Translation=(X=0,Y=-2,Z=4),Scale3D=(X=1,Y=1,Z=1))),(Item=(Type=Bone,Name=\"dyn_hair_back_03_l\"),Transform=(Rotation=(X=0,Y=0.0872,Z=0,W=0.9962),Translation=(X=0,Y=0 ...(truncated)"
    Weight=@n4.OutputParam
   )
   > n5 = &RigUnit_SphericalPoseReader(
    ActiveRegionScaleFactors=(PositiveWidth=1,NegativeWidth=2,PositiveHeight=2,NegativeHeight=2)
    ActiveRegionSize=0.1
    Debug=(bDrawDebug=true,bDraw2D=false,bDrawLocalAxes=false,DebugScale=25,DebugSegments=20,DebugThickness=0.25)
    DriverAxis=(X=-1,Y=0,Z=0)
    DriverItem=(Type=Bone,Name="clavicle_r")
    FalloffRegionScaleFactors=(PositiveWidth=1,NegativeWidth=0,PositiveHeight=0,NegativeHeight=0)
    FalloffSize=0.2
    OptionalParentItem=(Type=Bone)
    RotationOffset=(X=0,Y=130,Z=0)
   )
   > &RigUnit_ModifyTransforms(
    ItemToModify="((Item=(Type=Bone,Name=\"dyn_hair_back_02_r\"),Transform=(Rotation=(X=0,Y=-0.2588,Z=0,W=0.9659),Translation=(X=0,Y=-2,Z=-4),Scale3D=(X=1,Y=1,Z=1))),(Item=(Type=Bone,Name=\"dyn_hair_back_03_r\"),Transform=(Rotation=(X=0,Y=-0.061,Z=0,W=0.9981),Translation=(X=0,Y ...(truncated)"
    Weight=@n5.OutputParam
   )
  ]}
 SequenceAggregate_1_ContainedGraph {nodes [
   n1 = RigVMFunctionEntryNode()
   SequenceAggregate = &RigVMFunction_Sequence(ExecuteContext=@n1.ExecuteContext)
   SequenceAggregate_1 = &RigVMFunction_Sequence(ExecuteContext=@SequenceAggregate.B)
   SequenceAggregate_2 = &RigVMFunction_Sequence(ExecuteContext=@SequenceAggregate_1.B)
   RigVMFunctionReturnNode(A=@SequenceAggregate.A, B=@SequenceAggregate_1.A, C=@SequenceAggregate_2.A, D=@SequenceAggregate_2.B)
  ]}
]
hierarchy [
 root {loc=<0 0 0>, rot=<0 0 0 1>}
 pelvis {parent=root, loc=<0 -0.5301 98.0571>, rot=<0.0127 0.707 -0.0127 -0.707>}
 spine_01 {parent=pelvis, loc=<6.1208 -0.9328 0>, rot=<0 0 0.034 -0.9994>}
 spine_02 {parent=spine_01, loc=<6.5996 -0.2095 0>, rot=<0 0 0.0238 -0.9997>}
 spine_03 {parent=spine_02, loc=<6.4914 -0.1237 0>, rot=<0 0 0.0149 -0.9999>}
 spine_04 {parent=spine_03, loc=<6.9598 0.0794 0>, rot=<0 0 -0.0132 -0.9999>}
 spine_05 {parent=spine_04, loc=<6.9987 0 0>, rot=<0 0 -0.0015 -1>}
 neck_01 {parent=spine_05, loc=<11.3138 0.1529 0>, rot=<0 0 -0.2003 -0.9797>}
 neck_02 {parent=neck_01, loc=<4.5627 0 0>, rot=<0 0 0.0504 -0.9987>}
 head {parent=neck_02, loc=<5.3914 0 0>, rot=<0 0 0.115 -0.9934>}
 hair {parent=head, loc=<5 0 0>, rot=<0 0 0 1>}
 base_hair_front_r {parent=hair, loc=<1.0852 2.317 7.0191>, rot=<-0.0014 -1 0.0001 0.006>}
 dyn_hair_front_01_r {parent=base_hair_front_r, loc=<0 0 0>, rot=<0.0006 0.0162 -0.0275 -0.9995>}
 dyn_hair_front_02_r {parent=dyn_hair_front_01_r, loc=<5.9858 0 0>, rot=<-0.0003 -0.0029 -0.0521 -0.9986>}
 dyn_hair_front_03_r {parent=dyn_hair_front_02_r, loc=<6.0737 0 0>, rot=<0.0011 0.005 -0.0372 -0.9993>}
 dyn_hair_front_04_r {parent=dyn_hair_front_03_r, loc=<6.1615 0 0>, rot=<-0.0049 -0.0145 -0.0825 -0.9965>}
 dyn_hair_front_05_r {parent=dyn_hair_front_04_r, loc=<6.2452 0 0>, rot=<0 0 0 1>}
 base_hair_front_l {parent=hair, loc=<1.0841 2.4463 -7.0122>, rot=<-0.0558 -0.9983 0.0011 0.0177>}
 dyn_hair_front_01_l {parent=base_hair_front_l, loc=<0 0 0>, rot=<0.0008 0.009 0.0354 -0.9993>}
 dyn_hair_front_02_l {parent=dyn_hair_front_01_l, loc=<5.9882 0 0>, rot=<0 0.0003 -0.0585 -0.9983>}
 dyn_hair_front_03_l {parent=dyn_hair_front_02_l, loc=<6.061 0 0>, rot=<-0.0003 -0.0013 -0.0391 -0.9992>}
 dyn_hair_front_04_l {parent=dyn_hair_front_03_l, loc=<6.1669 0 0>, rot=<0.0041 0.0125 -0.0735 -0.9972>}
 dyn_hair_front_05_l {parent=dyn_hair_front_04_l, loc=<6.2358 0 0>, rot=<0 0 0 1>}
 base_hair_back_l {parent=hair, loc=<4.0094 -1.2366 -5.8543>, rot=<-0.0109 -0.9997 0.0005 0.024>}
 dyn_hair_back_01_l {parent=base_hair_back_l, loc=<0 0 0>, rot=<0.0034 0.0784 0.0432 -0.996>}
 dyn_hair_back_02_l {parent=dyn_hair_back_01_l, loc=<8.9031 0 0>, rot=<0 0 0.0402 -0.9992>}
 dyn_hair_back_03_l {parent=dyn_hair_back_02_l, loc=<6.484 0 0>, rot=<0 0 0.2542 -0.9671>}
 dyn_hair_back_04_l {parent=dyn_hair_back_03_l, loc=<7.8294 -0.0162 0>, rot=<0 0 -0.1451 -0.9894>}
 dyn_hair_back_05_l {parent=dyn_hair_back_04_l, loc=<7.3949 0 0>, rot=<0 0 0 1>}
 base_hair_back_m {parent=hair, loc=<3.7026 -4.9151 -0.0004>, rot=<-0.0048 -1 0 0>}
 dyn_hair_back_01_m {parent=base_hair_back_m, loc=<0 0 0>, rot=<0 0 0.0316 -0.9995>}
 dyn_hair_back_02_m {parent=dyn_hair_back_01_m, loc=<8.5451 0 0>, rot=<0 0 0.0804 -0.9968>}
 dyn_hair_back_03_m {parent=dyn_hair_back_02_m, loc=<6.1752 0.2378 0>, rot=<0 0 0.0668 -0.9978>}
 dyn_hair_back_04_m {parent=dyn_hair_back_03_m, loc=<6.5783 0 0>, rot=<0 0 -0.0259 -0.9997>}
 dyn_hair_back_05_m {parent=dyn_hair_back_04_m, loc=<7.3844 0 0>, rot=<0 0 0 -1>}
 base_hair_back_r {parent=hair, loc=<4.0094 -1.2366 5.8543>, rot=<0.0041 -1 0 0>}
 dyn_hair_back_01_r {parent=base_hair_back_r, loc=<0 0 0>, rot=<-0.0015 -0.0523 0.0282 -0.9982>}
 dyn_hair_back_02_r {parent=dyn_hair_back_01_r, loc=<8.9031 0 0>, rot=<0 0 0.0402 -0.9992>}
 dyn_hair_back_03_r {parent=dyn_hair_back_02_r, loc=<6.4837 0 0>, rot=<0 0 0.2542 -0.9671>}
 dyn_hair_back_04_r {parent=dyn_hair_back_03_r, loc=<7.808 0 0>, rot=<0 0 -0.1476 -0.9891>}
 dyn_hair_back_05_r {parent=dyn_hair_back_04_r, loc=<7.3949 0 0>, rot=<0 0 0 1>}
 base_hair_frontshort_r {parent=hair, loc=<4.0406 6.7723 4.8836>, rot=<-0.1183 -0.8986 -0.419 -0.0552>}
 dyn_hair_frontshort_01_r {parent=base_hair_frontshort_r, loc=<0 0 0>, rot=<0 0 0 1>}
 base_hair_frontshort_m {parent=hair, loc=<4.0206 9.1976 0.0272>, rot=<-0.1305 -0.9914 0 0>}
 dyn_hair_frontshort_01_m {parent=base_hair_frontshort_m, loc=<0 0 0>, rot=<0 0 0 1>}
 base_hair_frontshort_l {parent=hair, loc=<4.0392 6.9512 -4.8927>, rot=<-0.1183 -0.8986 0.419 0.0552>}
 dyn_hair_frontshort_01_l {parent=base_hair_frontshort_l, loc=<0 0 0>, rot=<0 0 0 1>}
 clavicle_l {parent=spine_05, loc=<5.4831 -0.1053 -3.2946>, rot=<0.0465 -0.7289 -0.009 -0.6829>}
 upperarm_l {parent=clavicle_l, loc=<9.8658 -1.2136 -0.4105>, rot=<0.0164 -0.3245 -0.0269 -0.9454>}
 upperarm_twist_01_l {parent=upperarm_l, loc=<7.2135 -0.1134 0>, rot=<0 0 0 1>}
 deform_upperarm_bk_base_l {parent=upperarm_twist_01_l, loc=<-6.4053 -4.3746 0.314>, rot=<0 0 0 1>}
 deform_upperarm_bk_l {parent=deform_upperarm_bk_base_l, loc=<0 0 0>, rot=<0 0 0 1>}
 deform_upperarm_fr_base_l {parent=upperarm_twist_01_l, loc=<-5.6312 5.092 0.3354>, rot=<0 0 0 1>}
 deform_upperarm_fr_l {parent=deform_upperarm_fr_base_l, loc=<0 0 0>, rot=<0 0 0 1>}
 deform_upperarm_up_base_l {parent=upperarm_twist_01_l, loc=<-8.891 0.4825 3.9517>, rot=<0 0 0 1>}
 deform_upperarm_up_l {parent=deform_upperarm_up_base_l, loc=<0 0 0>, rot=<0 0 0 1>}
 clavicle_r {parent=spine_05, loc=<5.4831 -0.1053 3.2946>, rot=<0.7289 0.0465 -0.6829 0.009>}
 upperarm_r {parent=clavicle_r, loc=<-9.8658 1.2136 0.4105>, rot=<0.0164 -0.3245 -0.0269 -0.9454>}
 upperarm_twist_01_r {parent=upperarm_r, loc=<-7.2135 0.1134 0>, rot=<0 0 0 -1>}
 deform_upperarm_bk_base_r {parent=upperarm_twist_01_r, loc=<6.4053 4.3746 -0.314>, rot=<0 0 0 -1>}
 deform_upperarm_bk_r {parent=deform_upperarm_bk_base_r, loc=<0 0 0>, rot=<0 0 0 1>}
 deform_upperarm_fr_base_r {parent=upperarm_twist_01_r, loc=<5.6312 -5.092 -0.3354>, rot=<0 0 0 -1>}
 deform_upperarm_fr_r {parent=deform_upperarm_fr_base_r, loc=<0 0 0>, rot=<0 0 0 1>}
 deform_upperarm_up_base_r {parent=upperarm_twist_01_r, loc=<8.891 -0.4825 -3.9517>, rot=<0 0 0 -1>}
 deform_upperarm_up_r {parent=deform_upperarm_up_base_r, loc=<0 0 0>, rot=<0 0 0 1>}
]
variables [parachuting_alpha {type=double, default=0}]
