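# Physics asset for the BoneMarrow medium female body's face accessory mesh, set up for two front hair strands to swing as jointed capsule chains.
# Each side chains a base and five dynamic hair bones from the head, with twist locked and swing limits that differ between left and right.
# Head, neck, upper spine, clavicle and upper-arm capsules, joined by free angular constraints, provide body shapes around the hair, likely as colliders.
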PhysicsAsset /Game/Characters/Player/Female/Medium/Bodies/F_MED_BoneMarrow/Meshes/Parts/F_MED_BoneMarrow_FaceAcc_Physics
bodies {
 base_hair_front_l [{shapeType=Capsule, length=0.5, radius=1, shapeName=base_hair_front_l_capsule, rotation=<rot 90 0 0>}]
 base_hair_front_r [{shapeType=Capsule, length=0.5, radius=1, shapeName=base_hair_front_r_capsule, rotation=<rot 90 0 0>}]
 clavicle_l [{shapeType=Capsule, length=8.5, radius=4, shapeName=clavicle_l_capsule, center=<4 0 0>, rotation=<rot 90 0 0>}]
 clavicle_r [{shapeType=Capsule, length=8.5, radius=4, shapeName=clavicle_r_capsule, center=<-4 0 0>, rotation=<rot 90 0 0>}]
 dyn_hair_front_01_l [{shapeType=Capsule, length=4, radius=1.5, shapeName=dyn_hair_front_01_l_capsule, center=<3.7205 0 0.43>, rotation=<rot 90 0 0>}]
 dyn_hair_front_01_r [{shapeType=Capsule, length=4, radius=1.5, shapeName=dyn_hair_front_01_r_capsule, center=<3.7205 0 -0.43>, rotation=<rot 90 0 0>}]
 dyn_hair_front_02_l [{shapeType=Capsule, length=4.75, radius=1.25, shapeName=dyn_hair_front_02_l_capsule, center=<3.8652 0 0>, rotation=<rot 90 0 0>}]
 dyn_hair_front_02_r [{shapeType=Capsule, length=4.5, radius=1.5, shapeName=dyn_hair_front_02_r_capsule, center=<3.8652 0 0>, rotation=<rot 90 0 0>}]
 dyn_hair_front_03_l [{shapeType=Capsule, length=4.75, radius=1.25, shapeName=dyn_hair_front_03_l_capsule, center=<4.0821 0 0>, rotation=<rot 90 0 0>}]
 dyn_hair_front_03_r [{shapeType=Capsule, length=4.25, radius=1, shapeName=dyn_hair_front_03_r_capsule, center=<4.0821 0 0>, rotation=<rot 90 0 0>}]
 dyn_hair_front_04_l [{shapeType=Capsule, length=3.25, radius=1.5, shapeName=dyn_hair_front_04_l_capsule, center=<2.8528 0 0>, rotation=<rot 90 0 0>}]
 dyn_hair_front_04_r [{shapeType=Capsule, length=3.5, radius=1.25, shapeName=dyn_hair_front_04_r_capsule, center=<2.8528 -0.2232 0>, rotation=<rot 90 0 0>}]
 dyn_hair_front_05_l [{shapeType=Capsule, length=5.5, radius=1.5, shapeName=dyn_hair_front_05_l_capsule, center=<4.142 0.8879 0>, rotation=<rot 90 0 0>}]
 dyn_hair_front_05_r [{shapeType=Capsule, length=5.5, radius=1.5, shapeName=dyn_hair_front_05_r_capsule, center=<4.142 0.8879 0>, rotation=<rot 90 0 0>}]
 head [
  {shapeType=Capsule, length=5, radius=5.8, center=<2.3363 2.3059 0>, rotation=<rot 90 0 0>}
  {shapeType=Capsule, length=66, radius=8, shapeName=neck_02_capsule, center=<1.0586 -8.2962 0>}
 ]
 neck_01 [
  {shapeType=Capsule, length=6.75, radius=2.505, shapeName=neck_01_capsule, center=<1.5954 5.2709 0>}
  {shapeType=Capsule, length=5, radius=5, shapeName=neck_01_capsule, center=<2.1269 0.3194 0>, rotation=<rot 90 0 0>}
 ]
 spine_05 [
  {shapeType=Capsule, length=9, radius=12.25, shapeName=spine_05_capsule, center=<-4.7092 -1.4061 0>, rotation=<rot 90 180 -139.9977>}
  {shapeType=Capsule, length=4, radius=11.25, shapeName=spine_05_capsule, center=<-6.7936 1.6897 0>}
 ]
 upperarm_l [{shapeType=Capsule, length=32.25, radius=4.25, shapeName=upperarm_l_capsule, center=<15.0285 -0.5624 0>, rotation=<rot 90 0 0>}]
 upperarm_r [{
   shapeType=Capsule
   length=33.25
   radius=4.75
   shapeName=upperarm_r_capsule
   center=<-15.0285 1.3524 0.3067>
   rotation=<rot 90 -0.0031 -0.0039>
  }]
}
constraints [
 {bone1Name=base_hair_front_l, bone2Name=head, swing1LimitDegrees=20, swing2LimitDegrees=10, twistLimitDegrees=5, twistMotion=Locked}
 {
  bone1Name=dyn_hair_front_01_l
  bone2Name=base_hair_front_l
  swing1LimitDegrees=20
  swing2LimitDegrees=8
  twistLimitDegrees=5
  twistMotion=Locked
 }
 {
  bone1Name=dyn_hair_front_02_l
  bone2Name=dyn_hair_front_01_l
  swing1LimitDegrees=23
  swing2LimitDegrees=8
  twistLimitDegrees=5
  twistMotion=Locked
 }
 {
  bone1Name=dyn_hair_front_03_l
  bone2Name=dyn_hair_front_02_l
  swing1LimitDegrees=23
  swing2LimitDegrees=8
  twistLimitDegrees=5
  twistMotion=Locked
 }
 {
  bone1Name=dyn_hair_front_04_l
  bone2Name=dyn_hair_front_03_l
  swing1LimitDegrees=25
  swing2LimitDegrees=20
  twistLimitDegrees=5
  twistMotion=Locked
 }
 {bone1Name=base_hair_front_r, bone2Name=head, swing1LimitDegrees=20, swing2LimitDegrees=10, twistLimitDegrees=5, twistMotion=Locked}
 {
  bone1Name=dyn_hair_front_01_r
  bone2Name=base_hair_front_r
  swing1LimitDegrees=20
  swing2LimitDegrees=16
  twistLimitDegrees=5
  twistMotion=Locked
 }
 {
  bone1Name=dyn_hair_front_02_r
  bone2Name=dyn_hair_front_01_r
  swing1LimitDegrees=20
  swing2LimitDegrees=16
  twistLimitDegrees=5
  twistMotion=Locked
 }
 {
  bone1Name=dyn_hair_front_03_r
  bone2Name=dyn_hair_front_02_r
  swing1LimitDegrees=20
  swing2LimitDegrees=16
  twistLimitDegrees=5
  twistMotion=Locked
 }
 {
  bone1Name=dyn_hair_front_04_r
  bone2Name=dyn_hair_front_03_r
  swing1LimitDegrees=30
  swing2LimitDegrees=30
  twistLimitDegrees=5
  twistMotion=Locked
 }
 {
  bone1Name=dyn_hair_front_05_l
  bone2Name=dyn_hair_front_04_l
  swing1LimitDegrees=25
  swing2LimitDegrees=16
  twistLimitDegrees=5
  twistMotion=Locked
 }
 {
  bone1Name=dyn_hair_front_05_r
  bone2Name=dyn_hair_front_04_r
  swing1LimitDegrees=25
  swing2LimitDegrees=20
  twistLimitDegrees=5
  twistMotion=Locked
 }
 {
  bone1Name=upperarm_l
  bone2Name=clavicle_l
  swing1LimitDegrees=45
  swing1Motion=Free
  swing2LimitDegrees=45
  swing2Motion=Free
  twistLimitDegrees=45
  twistMotion=Free
 }
 {
  bone1Name=upperarm_r
  bone2Name=clavicle_r
  swing1LimitDegrees=45
  swing1Motion=Free
  swing2LimitDegrees=45
  swing2Motion=Free
  twistLimitDegrees=45
  twistMotion=Free
 }
 {
  bone1Name=head
  bone2Name=neck_01
  swing1LimitDegrees=45
  swing1Motion=Free
  swing2LimitDegrees=45
  swing2Motion=Free
  twistLimitDegrees=45
  twistMotion=Free
 }
 {
  bone1Name=clavicle_l
  bone2Name=spine_05
  swing1LimitDegrees=45
  swing1Motion=Free
  swing2LimitDegrees=45
  swing2Motion=Free
  twistLimitDegrees=45
  twistMotion=Free
 }
 {
  bone1Name=clavicle_r
  bone2Name=spine_05
  swing1LimitDegrees=45
  swing1Motion=Free
  swing2LimitDegrees=45
  swing2Motion=Free
  twistLimitDegrees=45
  twistMotion=Free
 }
 {
  bone1Name=neck_01
  bone2Name=spine_05
  swing1LimitDegrees=45
  swing1Motion=Free
  swing2LimitDegrees=45
  swing2Motion=Free
  twistLimitDegrees=45
  twistMotion=Free
 }
]
