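# Physics asset for the back hair of the medium female BoneMarrow character's face accessory, simulating three hanging chains (left, middle, right).
# Each chain runs from a base bone locked to the head through five capsule segments with narrow swing limits and mostly locked twist.
# Angularly free constraints tie the outer tips to the middle tip, linking the chains.
# Capsules on head, neck, upper spine, clavicles and upper arms provide body collision shapes.
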PhysicsAsset /Game/Characters/Player/Female/Medium/Bodies/F_MED_BoneMarrow/Meshes/Parts/F_MED_BoneMarrow_FaceAcc_Physics_Back
bodies {
 base_hair_back_l [{shapeType=Capsule, length=0.5, radius=1, shapeName=base_hair_back_l_capsule, rotation=<rot 90 0 0>}]
 base_hair_back_m [{shapeType=Capsule, length=0.5, radius=1, shapeName=base_hair_back_m_capsule, rotation=<rot 90 0 0>}]
 base_hair_back_r [{shapeType=Capsule, length=0.5, radius=1, shapeName=base_hair_back_r_capsule, rotation=<rot 90 0 0>}]
 clavicle_l [{shapeType=Capsule, length=8, radius=5.5, shapeName=clavicle_l_capsule, center=<4 0 0>, rotation=<rot 90 0 0>}]
 clavicle_r [{shapeType=Capsule, length=8, radius=5.5, shapeName=clavicle_r_capsule, center=<-4 0 0>, rotation=<rot 90 0 0>}]
 dyn_hair_back_01_l [{shapeType=Capsule, length=5, radius=2, shapeName=dyn_hair_back_01_l_capsule, center=<4.4493 0 0>, rotation=<rot 90 0 0>}]
 dyn_hair_back_01_m [{shapeType=Capsule, length=5, radius=2, shapeName=dyn_hair_back_01_m_capsule, center=<4.038 0 0>, rotation=<rot 90 0 0>}]
 dyn_hair_back_01_r [{shapeType=Capsule, length=5, radius=2, shapeName=dyn_hair_back_01_r_capsule, center=<4.3344 0 0>, rotation=<rot 90 0 0>}]
 dyn_hair_back_02_l [{shapeType=Capsule, length=3, radius=2, shapeName=dyn_hair_back_02_l_capsule, center=<3.7719 0 0>, rotation=<rot 90 0 0>}]
 dyn_hair_back_02_m [{shapeType=Capsule, length=4, radius=2, shapeName=dyn_hair_back_02_m_capsule, center=<3.9431 0 0>, rotation=<rot 90 0 0>}]
 dyn_hair_back_02_r [{shapeType=Capsule, length=3, radius=2, shapeName=dyn_hair_back_02_r_capsule, center=<3.7719 0 0>, rotation=<rot 90 0 0>}]
 dyn_hair_back_03_l [{shapeType=Capsule, length=4, radius=2.5, shapeName=dyn_hair_back_03_l_capsule, center=<3.5056 0 0>, rotation=<rot 90 10 0>}]
 dyn_hair_back_03_m [{shapeType=Capsule, length=1.5, radius=2.7, shapeName=dyn_hair_back_03_m_capsule, center=<3.2794 0 0>, rotation=<rot 90 0 0>}]
 dyn_hair_back_03_r [{shapeType=Capsule, length=4, radius=2.5, shapeName=dyn_hair_back_03_r_capsule, center=<3.9838 0 0>, rotation=<rot 90 0 0>}]
 dyn_hair_back_04_l [{shapeType=Capsule, length=4, radius=2.5, shapeName=dyn_hair_back_04_l_capsule, center=<3.5887 0 0>, rotation=<rot 90 0 0>}]
 dyn_hair_back_04_m [{shapeType=Capsule, length=4, radius=2.7, shapeName=dyn_hair_back_04_m_capsule, center=<3.7441 0 0>, rotation=<rot 90 0 0>}]
 dyn_hair_back_04_r [{shapeType=Capsule, length=4, radius=2.5, shapeName=dyn_hair_back_04_r_capsule, center=<3.6397 0 0>, rotation=<rot 90 0 0>}]
 dyn_hair_back_05_l [{shapeType=Capsule, length=4.5, radius=3, shapeName=dyn_hair_back_04_l_capsule, center=<5 0 0>, rotation=<rot 90 0 0>}]
 dyn_hair_back_05_m [{shapeType=Capsule, length=4.5, radius=3, shapeName=dyn_hair_back_04_m_capsule, center=<5 0 0>, rotation=<rot 90 0 0>}]
 dyn_hair_back_05_r [{shapeType=Capsule, length=4.5, radius=3, shapeName=dyn_hair_back_04_r_capsule, center=<5 0 0>, rotation=<rot 90 0 0>}]
 head [{shapeType=Capsule, length=2, radius=6.8, shapeName=head_capsule, center=<5.5335 1.193 0>, rotation=<rot 90 0 60.0001>}]
 neck_01 [{shapeType=Capsule, length=6.75, radius=2.25, shapeName=neck_02_capsule, center=<0 2 0>}]
 spine_04 [{shapeType=Capsule, length=3, radius=11.5, shapeName=spine_04_capsule, center=<-0.4933 3.4188 0>}]
 spine_05 [
  {shapeType=Capsule, length=11, radius=7.5, shapeName=spine_05_capsule}
  {shapeType=Capsule, length=45.5, radius=13.25, shapeName=spine_05_capsule, center=<9.687 14.3909 0>}
 ]
 upperarm_l [{shapeType=Capsule, length=18, radius=5.5, shapeName=upperarm_l_capsule, center=<10 0 0>, rotation=<rot 90 0 0>}]
 upperarm_r [{shapeType=Capsule, length=18, radius=5.5, shapeName=upperarm_r_capsule, center=<-10 0 0>, rotation=<rot 90 0 0>}]
}
constraints [
 {
  bone1Name=base_hair_back_l
  bone2Name=head
  swing1LimitDegrees=10
  swing1Motion=Locked
  swing2LimitDegrees=10
  swing2Motion=Locked
  twistLimitDegrees=5
  twistMotion=Locked
 }
 {
  bone1Name=dyn_hair_back_01_l
  bone2Name=base_hair_back_l
  swing1LimitDegrees=10
  swing2LimitDegrees=10
  twistLimitDegrees=1
  twistMotion=Locked
 }
 {
  bone1Name=dyn_hair_back_02_l
  bone2Name=dyn_hair_back_01_l
  swing1LimitDegrees=10
  swing2LimitDegrees=10
  twistLimitDegrees=1
  twistMotion=Locked
 }
 {
  bone1Name=dyn_hair_back_03_l
  bone2Name=dyn_hair_back_02_l
  swing1LimitDegrees=10
  swing2LimitDegrees=8
  twistLimitDegrees=1
  twistMotion=Locked
 }
 {
  bone1Name=dyn_hair_back_04_l
  bone2Name=dyn_hair_back_03_l
  swing1LimitDegrees=12
  swing2LimitDegrees=5
  twistLimitDegrees=1
  twistMotion=Locked
 }
 {
  bone1Name=base_hair_back_m
  bone2Name=head
  swing1LimitDegrees=10
  swing1Motion=Locked
  swing2LimitDegrees=10
  swing2Motion=Locked
  twistLimitDegrees=5
  twistMotion=Locked
 }
 {
  bone1Name=dyn_hair_back_01_m
  bone2Name=base_hair_back_m
  swing1LimitDegrees=10
  swing2LimitDegrees=10
  twistLimitDegrees=1
  twistMotion=Locked
 }
 {
  bone1Name=dyn_hair_back_02_m
  bone2Name=dyn_hair_back_01_m
  swing1LimitDegrees=10
  swing2LimitDegrees=10
  twistLimitDegrees=1
  twistMotion=Locked
 }
 {bone1Name=dyn_hair_back_03_m, bone2Name=dyn_hair_back_02_m, swing1LimitDegrees=10, swing2LimitDegrees=8, twistLimitDegrees=1}
 {bone1Name=dyn_hair_back_04_m, bone2Name=dyn_hair_back_03_m, swing1LimitDegrees=12, swing2LimitDegrees=10, twistLimitDegrees=10}
 {
  bone1Name=base_hair_back_r
  bone2Name=head
  swing1LimitDegrees=10
  swing1Motion=Locked
  swing2LimitDegrees=10
  swing2Motion=Locked
  twistLimitDegrees=5
  twistMotion=Locked
 }
 {
  bone1Name=dyn_hair_back_01_r
  bone2Name=base_hair_back_r
  swing1LimitDegrees=10
  swing2LimitDegrees=10
  twistLimitDegrees=1
  twistMotion=Locked
 }
 {
  bone1Name=dyn_hair_back_02_r
  bone2Name=dyn_hair_back_01_r
  swing1LimitDegrees=10
  swing2LimitDegrees=10
  twistLimitDegrees=1
  twistMotion=Locked
 }
 {
  bone1Name=dyn_hair_back_03_r
  bone2Name=dyn_hair_back_02_r
  swing1LimitDegrees=10
  swing2LimitDegrees=8
  twistLimitDegrees=1
  twistMotion=Locked
 }
 {
  bone1Name=dyn_hair_back_04_r
  bone2Name=dyn_hair_back_03_r
  swing1LimitDegrees=12
  swing2LimitDegrees=5
  twistLimitDegrees=1
  twistMotion=Locked
 }
 {bone1Name=dyn_hair_back_05_l, bone2Name=dyn_hair_back_04_l, swing1LimitDegrees=25, swing2LimitDegrees=7, twistLimitDegrees=1}
 {bone1Name=dyn_hair_back_05_m, bone2Name=dyn_hair_back_04_m, swing1LimitDegrees=25, swing2LimitDegrees=7, twistLimitDegrees=1}
 {bone1Name=dyn_hair_back_05_r, bone2Name=dyn_hair_back_04_r, swing1LimitDegrees=25, swing2LimitDegrees=7, twistLimitDegrees=1}
 {
  bone1Name=dyn_hair_back_05_r
  bone2Name=dyn_hair_back_05_m
  swing1LimitDegrees=45
  swing1Motion=Free
  swing2LimitDegrees=45
  swing2Motion=Free
  twistLimitDegrees=45
  twistMotion=Free
 }
 {
  bone1Name=dyn_hair_back_05_l
  bone2Name=dyn_hair_back_05_m
  swing1LimitDegrees=45
  swing1Motion=Free
  swing2LimitDegrees=45
  swing2Motion=Free
  twistLimitDegrees=45
  twistMotion=Free
 }
]
