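# Animation Blueprint for the F_MED_BoneMarrow face accessory skeleton that adds secondary motion and body-relative placement to long hair bones.
# It takes a cached 'RBAN' pose into component space and constrains back and front strands to neck, spine and clavicle bones.
# Chained AnimDynamics nodes add spring motion, using a shared GravityOverride variable and spherical limits around body bones.
# Repeated additive bone offsets and constraint passes then adjust the back hair.
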AnimBlueprint /BRCosmetics/Characters/Player/Female/Medium/Bodies/F_MED_BoneMarrow/Meshes/Parts/F_MED_BoneMarrow_FaceAcc_AnimBP
target_skeleton=/Game/Characters/Player/Female/Medium/Bodies/F_MED_BoneMarrow/Meshes/Parts/F_MED_BoneMarrow_FaceAcc_Skeleton
graphs [
 AnimGraph {nodes [
   n1 = "Constraint - dyn_hair_back_03_r (PP)"!(ComponentPose=Local To Component(LocalPose=Use cached pose 'RBAN'()), Alpha=1, ConstraintWeights_0=0.7, ConstraintWeights_1:float=0.3) -> (Pose) {
    alphaBoolBlend {blendOption=Linear}
    boneToModify {boneName=dyn_hair_back_03_r}
    constraintSetup [{targetBone {boneName=neck_01}}, {targetBone {boneName=clavicle_r}}]
    constraintWeights [1,1]
   }
   n2 = "Constraint - dyn_hair_back_04_r (PP)"!(ComponentPose=@n1, Alpha=1, ConstraintWeights_0=0.5, ConstraintWeights_1:float=0.5) -> (Pose) {
    alphaBoolBlend {blendOption=Linear}
    boneToModify {boneName=dyn_hair_back_04_r}
    constraintSetup [{targetBone {boneName=spine_05}}, {targetBone {boneName=clavicle_r}}]
    constraintWeights [1,1]
   }
   n3 = "Constraint - dyn_hair_back_03_m (P)"(ComponentPose=@n2, Alpha=1, ConstraintWeights_0=1) {
    alphaBoolBlend {blendOption=Linear}
    boneToModify {boneName=dyn_hair_back_03_m}
    constraintSetup [{targetBone {boneName=neck_01}}]
    constraintWeights [1]
   }
   n4 = "Constraint - dyn_hair_back_04_m (P)"(ComponentPose=@n3, Alpha=1, ConstraintWeights_0=1) {
    alphaBoolBlend {blendOption=Linear}
    boneToModify {boneName=dyn_hair_back_04_m}
    constraintSetup [{targetBone {boneName=spine_05}}]
    constraintWeights [1]
   }
   n5 = "Constraint - dyn_hair_back_03_l (PP)"!(ComponentPose=@n4, Alpha=1, ConstraintWeights_0=0.7, ConstraintWeights_1:float=0.3) -> (Pose) {
    alphaBoolBlend {blendOption=Linear}
    boneToModify {boneName=dyn_hair_back_03_l}
    constraintSetup [{targetBone {boneName=neck_01}}, {targetBone {boneName=clavicle_l}}]
    constraintWeights [1,1]
   }
   n6 = "Constraint - dyn_hair_back_04_l (PP)"!(ComponentPose=@n5, Alpha=1, ConstraintWeights_0=0.5, ConstraintWeights_1:float=0.5) -> (Pose) {
    alphaBoolBlend {blendOption=Linear}
    boneToModify {boneName=dyn_hair_back_04_l}
    constraintSetup [{targetBone {boneName=spine_05}}, {targetBone {boneName=clavicle_l}}]
    constraintWeights [1,1]
   }
   n7 = "Constraint - dyn_hair_front_03_l (P)"(ComponentPose=@n6, Alpha=0.6, ConstraintWeights_0=1) {
    alphaBoolBlend {blendOption=Linear}
    boneToModify {boneName=dyn_hair_front_03_l}
    constraintSetup [{targetBone {boneName=clavicle_l}}]
    constraintWeights [1]
   }
   n8 = "Constraint - dyn_hair_front_03_r (P)"(ComponentPose=@n7, Alpha=0.6, ConstraintWeights_0=1) {
    alphaBoolBlend {blendOption=Linear}
    boneToModify {boneName=dyn_hair_front_03_r}
    constraintSetup [{targetBone {boneName=clavicle_r}}]
    constraintWeights [1]
   }
   n9 = Get GravityOverride()
   n10 = Anim Dynamics - Bone: dyn_hair_front_02_l!(ComponentPose=@n8, Alpha=1, GravityOverride:Vector=@n9, ExternalForce="0,0,0") -> (Pose) {
    angularDampingOverride=0.9999
    angularSpringConstant=500
    bAngularSpring=true
    bGravityOverrideInSimSpace=true
    bLinearSpring=true
    bOverrideAngularDamping=true
    bUseGravityOverride=true
    bUseSphericalLimits=true
    linearSpringConstant=50
    simulationSpace=BoneRelative
    alphaBoolBlend {blendOption=Linear}
    boundBone {boneName=dyn_hair_front_02_l}
    componentAppliedLinearAccClamp=<1500 3000 3000>
    componentLinearAccScale=<0.1 0.3 0.5>
    componentLinearVelScale=<0.1 0.3 0.5>
    gravityOverride {z=-980}
    physicsBodyDefinitions [{
      collisionType=CustomSphere
      sphereCollisionRadius=1.5
      boundBone {boneName=dyn_hair_front_02_l}
      boxExtents=<5 20 20>
      constraintSetup {angularLimitsMax=<1 5 3>, angularLimitsMin=<-1 -5 -10>, angularTarget=<0 0 0>}
      localJointOffset=<3 0 0>
     }]
    relativeSpaceBone {boneName=root}
    sphericalLimits [
     {limitRadius=3, drivingBone {boneName=neck_01}, sphereLocalOffset=<0 0 0>}
     {limitRadius=0, drivingBone {boneName=None}, sphereLocalOffset=<0 0 0>}
    ]
   }
   n11 = Anim Dynamics - Bone: dyn_hair_back_03_l!(ComponentPose=@n10, Alpha=1, GravityOverride:Vector=@n9, ExternalForce="0,0,0") -> (Pose) {
    angularDampingOverride=0.9999
    angularSpringConstant=500
    bAngularSpring=true
    bGravityOverrideInSimSpace=true
    bLinearSpring=true
    bOverrideAngularDamping=true
    bUseGravityOverride=true
    bUseSphericalLimits=true
    linearSpringConstant=50
    simulationSpace=BoneRelative
    alphaBoolBlend {blendOption=Linear}
    boundBone {boneName=dyn_hair_back_03_l}
    componentAppliedLinearAccClamp=<1500 3000 3000>
    componentLinearAccScale=<0.1 0.3 0.5>
    componentLinearVelScale=<0.1 0.3 0.5>
    gravityOverride {z=-980}
    physicsBodyDefinitions [{
      collisionType=CustomSphere
      sphereCollisionRadius=1.5
      boundBone {boneName=dyn_hair_back_03_l}
      boxExtents=<5 20 20>
      constraintSetup {angularLimitsMax=<1 5 3>, angularLimitsMin=<-1 -5 -10>, angularTarget=<0 0 0>}
      localJointOffset=<3 0 0>
     }]
    relativeSpaceBone {boneName=root}
    sphericalLimits [
     {limitRadius=3, drivingBone {boneName=neck_01}, sphereLocalOffset=<0 0 0>}
     {limitRadius=5, drivingBone {boneName=clavicle_l}, sphereLocalOffset=<0 0 0>}
    ]
   }
   n12 = Anim Dynamics - Bone: dyn_hair_back_04_l!(ComponentPose=@n11, Alpha=1, GravityOverride:Vector=@n9, ExternalForce="0,0,0") -> (Pose) {
    angularDampingOverride=0.9999
    angularSpringConstant=500
    bAngularSpring=true
    bGravityOverrideInSimSpace=true
    bLinearSpring=true
    bOverrideAngularDamping=true
    bUseGravityOverride=true
    bUseSphericalLimits=true
    linearSpringConstant=50
    simulationSpace=BoneRelative
    alphaBoolBlend {blendOption=Linear}
    boundBone {boneName=dyn_hair_back_04_l}
    componentAppliedLinearAccClamp=<1500 3000 3000>
    componentLinearAccScale=<0.1 0.3 0.5>
    componentLinearVelScale=<0.1 0.3 0.5>
    gravityOverride {z=-980}
    physicsBodyDefinitions [{
      collisionType=CustomSphere
      sphereCollisionRadius=1.5
      boundBone {boneName=dyn_hair_back_04_l}
      boxExtents=<5 20 20>
      constraintSetup {angularLimitsMax=<1 10 4>, angularLimitsMin=<-1 -10 -10>, angularTarget=<0 0 0>}
      localJointOffset=<3 0 0>
     }]
    relativeSpaceBone {boneName=root}
    sphericalLimits [
     {limitRadius=3, drivingBone {boneName=spine_05}, sphereLocalOffset=<0 0 0>}
     {limitRadius=5, drivingBone {boneName=clavicle_l}, sphereLocalOffset=<0 0 0>}
    ]
   }
   n13 = Anim Dynamics - Bone: dyn_hair_back_05_l!(ComponentPose=@n12, Alpha=1, GravityOverride:Vector=@n9, ExternalForce="0,0,0") -> (Pose) {
    angularDampingOverride=0.9999
    angularSpringConstant=500
    bAngularSpring=true
    bGravityOverrideInSimSpace=true
    bLinearSpring=true
    bOverrideAngularDamping=true
    bUseGravityOverride=true
    bUseSphericalLimits=true
    linearSpringConstant=50
    simulationSpace=BoneRelative
    alphaBoolBlend {blendOption=Linear}
    boundBone {boneName=dyn_hair_back_05_l}
    componentAppliedLinearAccClamp=<1500 3000 3000>
    componentLinearAccScale=<0.1 0.3 0.5>
    componentLinearVelScale=<0.1 0.3 0.5>
    gravityOverride {z=-980}
    physicsBodyDefinitions [{
      collisionType=CustomSphere
      sphereCollisionRadius=1.5
      boundBone {boneName=dyn_hair_back_05_l}
      boxExtents=<5 20 20>
      constraintSetup {angularLimitsMax=<1 10 4>, angularLimitsMin=<-1 -10 -10>, angularTarget=<0 0 0>}
      localJointOffset=<3 0 0>
     }]
    relativeSpaceBone {boneName=root}
    sphericalLimits [
     {limitRadius=3, drivingBone {boneName=spine_05}, sphereLocalOffset=<0 0 0>}
     {limitRadius=5, drivingBone {boneName=clavicle_l}, sphereLocalOffset=<0 0 0>}
    ]
   }
   n14 = Anim Dynamics - Bone: dyn_hair_back_02_m!(ComponentPose=@n13, Alpha=1, GravityOverride:Vector=@n9, ExternalForce="0,0,0") -> (Pose) {
    angularDampingOverride=0.9999
    angularSpringConstant=500
    bAngularSpring=true
    bGravityOverrideInSimSpace=true
    bLinearSpring=true
    bOverrideAngularDamping=true
    bUseGravityOverride=true
    bUseSphericalLimits=true
    linearSpringConstant=50
    simulationSpace=BoneRelative
    alphaBoolBlend {blendOption=Linear}
    boundBone {boneName=dyn_hair_back_02_m}
    componentAppliedLinearAccClamp=<1500 3000 3000>
    componentLinearAccScale=<0.1 0.3 0.5>
    componentLinearVelScale=<0.1 0.3 0.5>
    gravityOverride {z=-980}
    physicsBodyDefinitions [{
      collisionType=CustomSphere
      sphereCollisionRadius=1.5
      boundBone {boneName=dyn_hair_back_02_m}
      boxExtents=<5 20 20>
      constraintSetup {angularLimitsMax=<1 5 3>, angularLimitsMin=<-1 -5 -10>, angularTarget=<0 0 0>}
      localJointOffset=<3 0 0>
     }]
    relativeSpaceBone {boneName=root}
    sphericalLimits [
     {limitRadius=3, drivingBone {boneName=neck_01}, sphereLocalOffset=<0 0 0>}
     {limitRadius=2, drivingBone {boneName=spine_05}, sphereLocalOffset=<0 0 0>}
    ]
   }
   n15 = Anim Dynamics - Bone: dyn_hair_back_03_m!(ComponentPose=@n14, Alpha=1, GravityOverride:Vector=@n9, ExternalForce="0,0,0") -> (Pose) {
    angularDampingOverride=0.9999
    angularSpringConstant=500
    bAngularSpring=true
    bGravityOverrideInSimSpace=true
    bLinearSpring=true
    bOverrideAngularDamping=true
    bUseGravityOverride=true
    bUseSphericalLimits=true
    linearSpringConstant=50
    simulationSpace=BoneRelative
    alphaBoolBlend {blendOption=Linear}
    boundBone {boneName=dyn_hair_back_03_m}
    componentAppliedLinearAccClamp=<1500 3000 3000>
    componentLinearAccScale=<0.1 0.3 0.5>
    componentLinearVelScale=<0.1 0.3 0.5>
    gravityOverride {z=-980}
    physicsBodyDefinitions [{
      collisionType=CustomSphere
      sphereCollisionRadius=1.5
      boundBone {boneName=dyn_hair_back_03_m}
      boxExtents=<5 20 20>
      constraintSetup {angularLimitsMax=<1 5 3>, angularLimitsMin=<-1 -5 -10>, angularTarget=<0 0 0>}
      localJointOffset=<3 0 0>
     }]
    relativeSpaceBone {boneName=root}
    sphericalLimits [
     {limitRadius=3, drivingBone {boneName=neck_01}, sphereLocalOffset=<0 0 0>}
     {limitRadius=2, drivingBone {boneName=spine_05}, sphereLocalOffset=<0 0 0>}
    ]
   }
   n16 = Anim Dynamics - Bone: dyn_hair_back_04_m!(ComponentPose=@n15, Alpha=1, GravityOverride:Vector=@n9, ExternalForce="0,0,0") -> (Pose) {
    angularDampingOverride=0.9999
    angularSpringConstant=500
    bAngularSpring=true
    bGravityOverrideInSimSpace=true
    bLinearSpring=true
    bOverrideAngularDamping=true
    bUseGravityOverride=true
    bUseSphericalLimits=true
    linearSpringConstant=50
    simulationSpace=BoneRelative
    alphaBoolBlend {blendOption=Linear}
    boundBone {boneName=dyn_hair_back_04_m}
    componentAppliedLinearAccClamp=<1500 3000 3000>
    componentLinearAccScale=<0.1 0.3 0.5>
    componentLinearVelScale=<0.1 0.3 0.5>
    gravityOverride {z=-980}
    physicsBodyDefinitions [{
      collisionType=CustomSphere
      sphereCollisionRadius=1.5
      boundBone {boneName=dyn_hair_back_04_m}
      boxExtents=<5 20 20>
      constraintSetup {angularLimitsMax=<1 10 4>, angularLimitsMin=<-1 -10 -10>, angularTarget=<0 0 0>}
      localJointOffset=<3 0 0>
     }]
    relativeSpaceBone {boneName=root}
    sphericalLimits [
     {limitRadius=4, drivingBone {boneName=spine_05}, sphereLocalOffset=<0 0 0>}
     {limitRadius=0, drivingBone {boneName=None}, sphereLocalOffset=<0 0 0>}
    ]
   }
   n17 = Anim Dynamics - Bone: dyn_hair_back_05_m!(ComponentPose=@n16, Alpha=1, GravityOverride:Vector=@n9, ExternalForce="0,0,0") -> (Pose) {
    angularDampingOverride=0.9999
    angularSpringConstant=500
    bAngularSpring=true
    bGravityOverrideInSimSpace=true
    bLinearSpring=true
    bOverrideAngularDamping=true
    bUseGravityOverride=true
    bUseSphericalLimits=true
    linearSpringConstant=50
    simulationSpace=BoneRelative
    alphaBoolBlend {blendOption=Linear}
    boundBone {boneName=dyn_hair_back_05_m}
    componentAppliedLinearAccClamp=<1500 3000 3000>
    componentLinearAccScale=<0.1 0.3 0.5>
    componentLinearVelScale=<0.1 0.3 0.5>
    gravityOverride {z=-980}
    physicsBodyDefinitions [{
      collisionType=CustomSphere
      sphereCollisionRadius=1.5
      boundBone {boneName=dyn_hair_back_05_m}
      boxExtents=<5 20 20>
      constraintSetup {angularLimitsMax=<1 10 4>, angularLimitsMin=<-1 -10 -10>, angularTarget=<0 0 0>}
      localJointOffset=<3 0 0>
     }]
    relativeSpaceBone {boneName=root}
    sphericalLimits [
     {limitRadius=4, drivingBone {boneName=spine_05}, sphereLocalOffset=<0 0 0>}
     {limitRadius=0, drivingBone {boneName=None}, sphereLocalOffset=<0 0 0>}
    ]
   }
   n18 = Anim Dynamics - Bone: dyn_hair_back_02_r!(ComponentPose=@n17, Alpha=1, GravityOverride:Vector=@n9, ExternalForce="0,0,0") -> (Pose) {
    angularDampingOverride=0.9999
    angularSpringConstant=500
    bAngularSpring=true
    bGravityOverrideInSimSpace=true
    bLinearSpring=true
    bOverrideAngularDamping=true
    bUseGravityOverride=true
    bUseSphericalLimits=true
    linearSpringConstant=50
    simulationSpace=BoneRelative
    alphaBoolBlend {blendOption=Linear}
    boundBone {boneName=dyn_hair_back_02_r}
    componentAppliedLinearAccClamp=<1500 3000 3000>
    componentLinearAccScale=<0.1 0.3 0.5>
    componentLinearVelScale=<0.1 0.3 0.5>
    gravityOverride {z=-980}
    physicsBodyDefinitions [{
      collisionType=CustomSphere
      sphereCollisionRadius=1.5
      boundBone {boneName=dyn_hair_back_02_r}
      boxExtents=<5 20 20>
      constraintSetup {angularLimitsMax=<1 5 3>, angularLimitsMin=<-1 -5 -10>, angularTarget=<0 0 0>}
      localJointOffset=<3 0 0>
     }]
    relativeSpaceBone {boneName=root}
    sphericalLimits [
     {limitRadius=3, drivingBone {boneName=neck_01}, sphereLocalOffset=<0 0 0>}
     {limitRadius=0, drivingBone {boneName=None}, sphereLocalOffset=<0 0 0>}
    ]
   }
   n19 = Anim Dynamics - Bone: dyn_hair_back_03_r!(ComponentPose=@n18, Alpha=1, GravityOverride:Vector=@n9, ExternalForce="0,0,0") -> (Pose) {
    angularDampingOverride=0.9999
    angularSpringConstant=500
    bAngularSpring=true
    bGravityOverrideInSimSpace=true
    bLinearSpring=true
    bOverrideAngularDamping=true
    bUseGravityOverride=true
    bUseSphericalLimits=true
    linearSpringConstant=50
    simulationSpace=BoneRelative
    alphaBoolBlend {blendOption=Linear}
    boundBone {boneName=dyn_hair_back_03_r}
    componentAppliedLinearAccClamp=<1500 3000 3000>
    componentLinearAccScale=<0.1 0.3 0.5>
    componentLinearVelScale=<0.1 0.3 0.5>
    gravityOverride {z=-980}
    physicsBodyDefinitions [{
      collisionType=CustomSphere
      sphereCollisionRadius=1.5
      boundBone {boneName=dyn_hair_back_03_r}
      boxExtents=<5 20 20>
      constraintSetup {angularLimitsMax=<1 5 3>, angularLimitsMin=<-1 -5 -10>, angularTarget=<0 0 0>}
      localJointOffset=<3 0 0>
     }]
    relativeSpaceBone {boneName=root}
    sphericalLimits [
     {limitRadius=3, drivingBone {boneName=neck_01}, sphereLocalOffset=<0 0 0>}
     {limitRadius=5, drivingBone {boneName=clavicle_r}, sphereLocalOffset=<0 0 0>}
    ]
   }
   n20 = Anim Dynamics - Bone: dyn_hair_back_04_r!(ComponentPose=@n19, Alpha=1, GravityOverride:Vector=@n9, ExternalForce="0,0,0") -> (Pose) {
    angularDampingOverride=0.9999
    angularSpringConstant=500
    bAngularSpring=true
    bGravityOverrideInSimSpace=true
    bLinearSpring=true
    bOverrideAngularDamping=true
    bUseGravityOverride=true
    bUseSphericalLimits=true
    linearSpringConstant=50
    simulationSpace=BoneRelative
    alphaBoolBlend {blendOption=Linear}
    boundBone {boneName=dyn_hair_back_04_r}
    componentAppliedLinearAccClamp=<1500 3000 3000>
    componentLinearAccScale=<0.1 0.3 0.5>
    componentLinearVelScale=<0.1 0.3 0.5>
    gravityOverride {z=-980}
    physicsBodyDefinitions [{
      collisionType=CustomSphere
      sphereCollisionRadius=1.5
      boundBone {boneName=dyn_hair_back_04_r}
      boxExtents=<5 20 20>
      constraintSetup {angularLimitsMax=<1 10 4>, angularLimitsMin=<-1 -10 -10>, angularTarget=<0 0 0>}
      localJointOffset=<3 0 0>
     }]
    relativeSpaceBone {boneName=root}
    sphericalLimits [
     {limitRadius=3, drivingBone {boneName=spine_05}, sphereLocalOffset=<0 0 0>}
     {limitRadius=5, drivingBone {boneName=clavicle_r}, sphereLocalOffset=<0 0 0>}
    ]
   }
   n21 = Anim Dynamics - Bone: dyn_hair_back_05_r!(ComponentPose=@n20, Alpha=1, GravityOverride:Vector=@n9, ExternalForce="0,0,0") -> (Pose) {
    angularDampingOverride=0.9999
    angularSpringConstant=500
    bAngularSpring=true
    bGravityOverrideInSimSpace=true
    bLinearSpring=true
    bOverrideAngularDamping=true
    bUseGravityOverride=true
    bUseSphericalLimits=true
    linearSpringConstant=50
    simulationSpace=BoneRelative
    alphaBoolBlend {blendOption=Linear}
    boundBone {boneName=dyn_hair_back_05_r}
    componentAppliedLinearAccClamp=<1500 3000 3000>
    componentLinearAccScale=<0.1 0.3 0.5>
    componentLinearVelScale=<0.1 0.3 0.5>
    gravityOverride {z=-980}
    physicsBodyDefinitions [{
      collisionType=CustomSphere
      sphereCollisionRadius=1.5
      boundBone {boneName=dyn_hair_back_05_r}
      boxExtents=<5 20 20>
      constraintSetup {angularLimitsMax=<1 10 4>, angularLimitsMin=<-1 -10 -10>, angularTarget=<0 0 0>}
      localJointOffset=<3 0 0>
     }]
    relativeSpaceBone {boneName=root}
    sphericalLimits [
     {limitRadius=3, drivingBone {boneName=spine_05}, sphereLocalOffset=<0 0 0>}
     {limitRadius=5, drivingBone {boneName=clavicle_r}, sphereLocalOffset=<0 0 0>}
    ]
   }
   n22 = Get GravityOverride()
   n23 = Anim Dynamics - Bone: dyn_hair_front_02_l(ComponentPose=@n21, Alpha=1, GravityOverride=@n22, ExternalForce="0,0,0") {
    angularDampingOverride=0.9999
    angularSpringConstant=500
    bAngularSpring=true
    bGravityOverrideInSimSpace=true
    bLinearSpring=true
    bOverrideAngularDamping=true
    bUseGravityOverride=true
    bUseSphericalLimits=true
    linearSpringConstant=50
    simulationSpace=BoneRelative
    alphaBoolBlend {blendOption=Linear}
    boundBone {boneName=dyn_hair_front_02_l}
    componentAppliedLinearAccClamp=<1500 3000 3000>
    componentLinearAccScale=<0.1 0.3 0.5>
    componentLinearVelScale=<0.1 0.3 0.5>
    gravityOverride {z=-980}
    physicsBodyDefinitions [{
      collisionType=CustomSphere
      sphereCollisionRadius=1.5
      boundBone {boneName=dyn_hair_front_02_l}
      boxExtents=<5 20 20>
      constraintSetup {angularLimitsMax=<1 2 5>, angularLimitsMin=<-1 -2 -5>, angularTarget=<0 0 0>}
      localJointOffset=<3 0 0>
     }]
    relativeSpaceBone {boneName=root}
    sphericalLimits [
     {limitRadius=8, drivingBone {boneName=spine_05}, sphereLocalOffset=<0 0 0>}
     {limitRadius=8.5, drivingBone {boneName=clavicle_l}, sphereLocalOffset=<0 0 0>}
     {limitRadius=7, drivingBone {boneName=upperarm_l}, sphereLocalOffset=<0 0 0>}
    ]
   }
   n24 = Anim Dynamics - Bone: dyn_hair_front_03_l!(ComponentPose=@n23, Alpha=1, GravityOverride:Vector=@n22, ExternalForce="0,0,0") -> (Pose) {
    angularDampingOverride=0.9999
    angularSpringConstant=500
    bAngularSpring=true
    bGravityOverrideInSimSpace=true
    bLinearSpring=true
    bOverrideAngularDamping=true
    bUseGravityOverride=true
    bUseSphericalLimits=true
    linearSpringConstant=50
    simulationSpace=BoneRelative
    alphaBoolBlend {blendOption=Linear}
    boundBone {boneName=dyn_hair_front_03_l}
    componentAppliedLinearAccClamp=<1500 3000 3000>
    componentLinearAccScale=<0.1 0.3 0.5>
    componentLinearVelScale=<0.1 0.3 0.5>
    gravityOverride {z=-980}
    physicsBodyDefinitions [{
      collisionType=CustomSphere
      sphereCollisionRadius=1.5
      boundBone {boneName=dyn_hair_front_03_l}
      boxExtents=<5 20 20>
      constraintSetup {angularLimitsMax=<1 2 5>, angularLimitsMin=<-1 -2 -5>, angularTarget=<0 0 0>}
      localJointOffset=<3 0 0>
     }]
    relativeSpaceBone {boneName=root}
    sphericalLimits [
     {limitRadius=8, drivingBone {boneName=spine_05}, sphereLocalOffset=<0 0 0>}
     {limitRadius=8.5, drivingBone {boneName=clavicle_l}, sphereLocalOffset=<0 0 0>}
     {limitRadius=7, drivingBone {boneName=upperarm_l}, sphereLocalOffset=<0 0 0>}
    ]
   }
   n25 = Anim Dynamics - Bone: dyn_hair_front_04_l!(ComponentPose=@n24, Alpha=1, GravityOverride:Vector=@n22, ExternalForce="0,0,0") -> (Pose) {
    angularDampingOverride=0.9999
    angularSpringConstant=500
    bAngularSpring=true
    bGravityOverrideInSimSpace=true
    bLinearSpring=true
    bOverrideAngularDamping=true
    bUseGravityOverride=true
    bUseSphericalLimits=true
    linearSpringConstant=50
    simulationSpace=BoneRelative
    alphaBoolBlend {blendOption=Linear}
    boundBone {boneName=dyn_hair_front_04_l}
    componentAppliedLinearAccClamp=<1500 3000 3000>
    componentLinearAccScale=<0.1 0.3 0.5>
    componentLinearVelScale=<0.1 0.3 0.5>
    gravityOverride {z=-980}
    physicsBodyDefinitions [{
      collisionType=CustomSphere
      sphereCollisionRadius=1.5
      boundBone {boneName=dyn_hair_front_04_l}
      boxExtents=<5 20 20>
      constraintSetup {angularLimitsMax=<1 5 10>, angularLimitsMin=<-1 -5 -10>, angularTarget=<0 0 0>}
      localJointOffset=<3 0 0>
     }]
    relativeSpaceBone {boneName=root}
    sphericalLimits [
     {limitRadius=8, drivingBone {boneName=spine_05}, sphereLocalOffset=<0 0 0>}
     {limitRadius=8.5, drivingBone {boneName=clavicle_l}, sphereLocalOffset=<0 0 0>}
     {limitRadius=7, drivingBone {boneName=upperarm_l}, sphereLocalOffset=<0 0 0>}
    ]
   }
   n26 = Anim Dynamics - Bone: dyn_hair_front_05_l!(ComponentPose=@n25, Alpha=1, GravityOverride:Vector=@n22, ExternalForce="0,0,0") -> (Pose) {
    angularDampingOverride=0.9999
    angularSpringConstant=500
    bAngularSpring=true
    bGravityOverrideInSimSpace=true
    bLinearSpring=true
    bOverrideAngularDamping=true
    bUseGravityOverride=true
    bUseSphericalLimits=true
    linearSpringConstant=50
    simulationSpace=BoneRelative
    alphaBoolBlend {blendOption=Linear}
    boundBone {boneName=dyn_hair_front_05_l}
    componentAppliedLinearAccClamp=<1500 3000 3000>
    componentLinearAccScale=<0.1 0.3 0.5>
    componentLinearVelScale=<0.1 0.3 0.5>
    gravityOverride {z=-980}
    physicsBodyDefinitions [{
      collisionType=CustomSphere
      sphereCollisionRadius=1.5
      boundBone {boneName=dyn_hair_front_05_l}
      boxExtents=<5 20 20>
      constraintSetup {angularLimitsMax=<1 5 10>, angularLimitsMin=<-1 -5 -10>, angularTarget=<0 0 0>}
      localJointOffset=<3 0 0>
     }]
    relativeSpaceBone {boneName=root}
    sphericalLimits [
     {limitRadius=8, drivingBone {boneName=spine_05}, sphereLocalOffset=<0 0 0>}
     {limitRadius=8.5, drivingBone {boneName=clavicle_l}, sphereLocalOffset=<0 0 0>}
     {limitRadius=7, drivingBone {boneName=upperarm_l}, sphereLocalOffset=<0 0 0>}
    ]
   }
   n27 = Anim Dynamics - Bone: dyn_hair_front_02_r!(ComponentPose=@n26, Alpha=1, GravityOverride:Vector=@n22, ExternalForce="0,0,0") -> (Pose) {
    angularDampingOverride=0.9999
    angularSpringConstant=500
    bAngularSpring=true
    bGravityOverrideInSimSpace=true
    bLinearSpring=true
    bOverrideAngularDamping=true
    bUseGravityOverride=true
    bUseSphericalLimits=true
    linearSpringConstant=50
    simulationSpace=BoneRelative
    alphaBoolBlend {blendOption=Linear}
    boundBone {boneName=dyn_hair_front_02_r}
    componentAppliedLinearAccClamp=<1500 3000 3000>
    componentLinearAccScale=<0.1 0.3 0.5>
    componentLinearVelScale=<0.1 0.3 0.5>
    gravityOverride {z=-980}
    physicsBodyDefinitions [{
      collisionType=CustomSphere
      sphereCollisionRadius=1.5
      boundBone {boneName=dyn_hair_front_02_r}
      boxExtents=<5 20 20>
      constraintSetup {angularLimitsMax=<1 2 5>, angularLimitsMin=<-1 -2 -5>, angularTarget=<0 0 0>}
      localJointOffset=<3 0 0>
     }]
    relativeSpaceBone {boneName=root}
    sphericalLimits [
     {limitRadius=8, drivingBone {boneName=spine_05}, sphereLocalOffset=<0 0 0>}
     {limitRadius=8.5, drivingBone {boneName=clavicle_r}, sphereLocalOffset=<0 0 0>}
     {limitRadius=7, drivingBone {boneName=upperarm_r}, sphereLocalOffset=<0 0 0>}
    ]
   }
   n28 = Anim Dynamics - Bone: dyn_hair_front_03_r!(ComponentPose=@n27, Alpha=1, GravityOverride:Vector=@n22, ExternalForce="0,0,0") -> (Pose) {
    angularDampingOverride=0.9999
    angularSpringConstant=500
    bAngularSpring=true
    bGravityOverrideInSimSpace=true
    bLinearSpring=true
    bOverrideAngularDamping=true
    bUseGravityOverride=true
    bUseSphericalLimits=true
    linearSpringConstant=50
    simulationSpace=BoneRelative
    alphaBoolBlend {blendOption=Linear}
    boundBone {boneName=dyn_hair_front_03_r}
    componentAppliedLinearAccClamp=<1500 3000 3000>
    componentLinearAccScale=<0.1 0.3 0.5>
    componentLinearVelScale=<0.1 0.3 0.5>
    gravityOverride {z=-980}
    physicsBodyDefinitions [{
      collisionType=CustomSphere
      sphereCollisionRadius=1.5
      boundBone {boneName=dyn_hair_front_03_r}
      boxExtents=<5 20 20>
      constraintSetup {angularLimitsMax=<1 2 5>, angularLimitsMin=<-1 -2 -5>, angularTarget=<0 0 0>}
      localJointOffset=<3 0 0>
     }]
    relativeSpaceBone {boneName=root}
    sphericalLimits [
     {limitRadius=8, drivingBone {boneName=spine_05}, sphereLocalOffset=<0 0 0>}
     {limitRadius=8.5, drivingBone {boneName=clavicle_r}, sphereLocalOffset=<0 0 0>}
     {limitRadius=7, drivingBone {boneName=upperarm_r}, sphereLocalOffset=<0 0 0>}
    ]
   }
   n29 = Anim Dynamics - Bone: dyn_hair_front_04_r!(ComponentPose=@n28, Alpha=1, GravityOverride:Vector=@n22, ExternalForce="0,0,0") -> (Pose) {
    angularDampingOverride=0.9999
    angularSpringConstant=500
    bAngularSpring=true
    bGravityOverrideInSimSpace=true
    bLinearSpring=true
    bOverrideAngularDamping=true
    bUseGravityOverride=true
    bUseSphericalLimits=true
    linearSpringConstant=50
    simulationSpace=BoneRelative
    alphaBoolBlend {blendOption=Linear}
    boundBone {boneName=dyn_hair_front_04_r}
    componentAppliedLinearAccClamp=<1500 3000 3000>
    componentLinearAccScale=<0.1 0.3 0.5>
    componentLinearVelScale=<0.1 0.3 0.5>
    gravityOverride {z=-980}
    physicsBodyDefinitions [{
      collisionType=CustomSphere
      sphereCollisionRadius=1.5
      boundBone {boneName=dyn_hair_front_04_r}
      boxExtents=<5 20 20>
      constraintSetup {angularLimitsMax=<1 5 10>, angularLimitsMin=<-1 -5 -10>, angularTarget=<0 0 0>}
      localJointOffset=<3 0 0>
     }]
    relativeSpaceBone {boneName=root}
    sphericalLimits [
     {limitRadius=8, drivingBone {boneName=spine_05}, sphereLocalOffset=<0 0 0>}
     {limitRadius=8.5, drivingBone {boneName=clavicle_r}, sphereLocalOffset=<0 0 0>}
     {limitRadius=7, drivingBone {boneName=upperarm_r}, sphereLocalOffset=<0 0 0>}
    ]
   }
   n30 = Anim Dynamics - Bone: dyn_hair_front_05_r!(ComponentPose=@n29, Alpha=1, GravityOverride:Vector=@n22, ExternalForce="0,0,0") -> (Pose) {
    angularDampingOverride=0.9999
    angularSpringConstant=500
    bAngularSpring=true
    bGravityOverrideInSimSpace=true
    bLinearSpring=true
    bOverrideAngularDamping=true
    bUseGravityOverride=true
    bUseSphericalLimits=true
    linearSpringConstant=50
    simulationSpace=BoneRelative
    alphaBoolBlend {blendOption=Linear}
    boundBone {boneName=dyn_hair_front_05_r}
    componentAppliedLinearAccClamp=<1500 3000 3000>
    componentLinearAccScale=<0.1 0.3 0.5>
    componentLinearVelScale=<0.1 0.3 0.5>
    gravityOverride {z=-980}
    physicsBodyDefinitions [{
      collisionType=CustomSphere
      sphereCollisionRadius=1.5
      boundBone {boneName=dyn_hair_front_05_r}
      boxExtents=<5 20 20>
      constraintSetup {angularLimitsMax=<1 5 10>, angularLimitsMin=<-1 -5 -10>, angularTarget=<0 0 0>}
      localJointOffset=<3 0 0>
     }]
    relativeSpaceBone {boneName=root}
    sphericalLimits [
     {limitRadius=8, drivingBone {boneName=spine_05}, sphereLocalOffset=<0 0 0>}
     {limitRadius=8.5, drivingBone {boneName=clavicle_r}, sphereLocalOffset=<0 0 0>}
     {limitRadius=7, drivingBone {boneName=upperarm_r}, sphereLocalOffset=<0 0 0>}
    ]
   }
   n31 = "Transform (Modify) Bone - Bone: dyn_hair_back_02_l"!(ComponentPose=@n30, Alpha=1, Rotation="-5,-10,0") -> (Pose) {
    rotationMode=BMM_Additive
    rotationSpace=BCS_BoneSpace
    alphaBoolBlend {blendOption=Linear}
    boneToModify {boneName=dyn_hair_back_02_l}
    rotation {pitch=-5, yaw=-10}
   }
   n32 = "Transform (Modify) Bone - Bone: dyn_hair_back_02_r"!(ComponentPose=@n31, Alpha=1, Rotation="5,-10,0") -> (Pose) {
    rotationMode=BMM_Additive
    rotationSpace=BCS_BoneSpace
    alphaBoolBlend {blendOption=Linear}
    boneToModify {boneName=dyn_hair_back_02_r}
    rotation {pitch=5, yaw=-10}
   }
   n33 = "Transform (Modify) Bone - Bone: dyn_hair_back_02_m"!(ComponentPose=@n32, Alpha=1, Rotation="0,-10,0") -> (Pose) {
    rotationMode=BMM_Additive
    rotationSpace=BCS_BoneSpace
    alphaBoolBlend {blendOption=Linear}
    boneToModify {boneName=dyn_hair_back_02_m}
    rotation {yaw=-10}
   }
   n34 = "Constraint - dyn_hair_back_02_l (P)"(ComponentPose=@n33, Alpha=1, ConstraintWeights_0=1) {
    alphaBoolBlend {blendOption=Linear}
    boneToModify {boneName=dyn_hair_back_02_l}
    constraintSetup [{targetBone {boneName=head}}]
    constraintWeights [1]
   }
   n35 = "Constraint - dyn_hair_back_02_m (P)"(ComponentPose=@n34, Alpha=1, ConstraintWeights_0=1) {
    alphaBoolBlend {blendOption=Linear}
    boneToModify {boneName=dyn_hair_back_02_m}
    constraintSetup [{targetBone {boneName=head}}]
    constraintWeights [1]
   }
   n36 = "Constraint - dyn_hair_front_02_r (P)"(ComponentPose=@n35, Alpha=1, ConstraintWeights_0=1) {
    alphaBoolBlend {blendOption=Linear}
    boneToModify {boneName=dyn_hair_front_02_r}
    constraintSetup [{targetBone {boneName=head}}]
    constraintWeights [1]
   }
   n37 = "Constraint - dyn_hair_back_03_r (PP)"(ComponentPose=@n36, Alpha=1, ConstraintWeights_0=0.7, ConstraintWeights_1=0.3) {
    alphaBoolBlend {blendOption=Linear}
    boneToModify {boneName=dyn_hair_back_03_r}
    constraintSetup [{targetBone {boneName=neck_01}}, {targetBone {boneName=clavicle_r}}]
    constraintWeights [1,1]
   }
   n38 = "Constraint - dyn_hair_back_04_r (PP)"(ComponentPose=@n37, Alpha=1, ConstraintWeights_0=0.5, ConstraintWeights_1=0.5) {
    alphaBoolBlend {blendOption=Linear}
    boneToModify {boneName=dyn_hair_back_04_r}
    constraintSetup [{targetBone {boneName=spine_05}}, {targetBone {boneName=clavicle_r}}]
    constraintWeights [1,1]
   }
   n39 = "Constraint - dyn_hair_back_03_m (P)"(ComponentPose=@n38, Alpha=1, ConstraintWeights_0=1) {
    alphaBoolBlend {blendOption=Linear}
    boneToModify {boneName=dyn_hair_back_03_m}
    constraintSetup [{targetBone {boneName=neck_01}}]
    constraintWeights [1]
   }
   n40 = "Constraint - dyn_hair_back_04_m (P)"(ComponentPose=@n39, Alpha=1, ConstraintWeights_0=1) {
    alphaBoolBlend {blendOption=Linear}
    boneToModify {boneName=dyn_hair_back_04_m}
    constraintSetup [{targetBone {boneName=spine_05}}]
    constraintWeights [1]
   }
   n41 = "Constraint - dyn_hair_back_03_l (PP)"(ComponentPose=@n40, Alpha=1, ConstraintWeights_0=0.7, ConstraintWeights_1=0.3) {
    alphaBoolBlend {blendOption=Linear}
    boneToModify {boneName=dyn_hair_back_03_l}
    constraintSetup [{targetBone {boneName=neck_01}}, {targetBone {boneName=clavicle_l}}]
    constraintWeights [1,1]
   }
   n42 = "Constraint - dyn_hair_back_04_l (PP)"(ComponentPose=@n41, Alpha=1, ConstraintWeights_0=0.5, ConstraintWeights_1=0.5) {
    alphaBoolBlend {blendOption=Linear}
    boneToModify {boneName=dyn_hair_back_04_l}
    constraintSetup [{targetBone {boneName=spine_05}}, {targetBone {boneName=clavicle_l}}]
    constraintWeights [1,1]
   }
   n43 = "Transform (Modify) Bone - Bone: dyn_hair_back_02_l"!(ComponentPose=@n42, Alpha=1, Translation:Vector="0,-3,0", Rotation="-5,-5,0") -> (Pose) {
    rotationMode=BMM_Additive
    rotationSpace=BCS_BoneSpace
    translationMode=BMM_Additive
    translationSpace=BCS_BoneSpace
    alphaBoolBlend {blendOption=Linear}
    boneToModify {boneName=dyn_hair_back_02_l}
    rotation {pitch=-5, yaw=-5}
    translation {y=-3}
   }
   n44 = "Transform (Modify) Bone - Bone: dyn_hair_back_02_r"!(ComponentPose=@n43, Alpha=1, Translation:Vector="0,-3,0", Rotation="5,-5,0") -> (Pose) {
    rotationMode=BMM_Additive
    rotationSpace=BCS_BoneSpace
    translationMode=BMM_Additive
    translationSpace=BCS_BoneSpace
    alphaBoolBlend {blendOption=Linear}
    boneToModify {boneName=dyn_hair_back_02_r}
    rotation {pitch=5, yaw=-5}
    translation {y=-3}
   }
   n45 = "Transform (Modify) Bone - Bone: dyn_hair_back_02_m"!(ComponentPose=@n44, Alpha=1, Translation:Vector="0,-3,0", Rotation="0,-5,0") -> (Pose) {
    rotationMode=BMM_Additive
    rotationSpace=BCS_BoneSpace
    translationMode=BMM_Additive
    translationSpace=BCS_BoneSpace
    alphaBoolBlend {blendOption=Linear}
    boneToModify {boneName=dyn_hair_back_02_m}
    rotation {yaw=-5}
    translation {y=-3}
   }
   n46 = "Constraint - dyn_hair_back_02_l (P)"(ComponentPose=@n45, Alpha=1, ConstraintWeights_0=1) {
    alphaBoolBlend {blendOption=Linear}
    boneToModify {boneName=dyn_hair_back_02_l}
    constraintSetup [{targetBone {boneName=head}}]
    constraintWeights [1]
   }
   n47 = "Constraint - dyn_hair_back_02_m (P)"(ComponentPose=@n46, Alpha=1, ConstraintWeights_0=1) {
    alphaBoolBlend {blendOption=Linear}
    boneToModify {boneName=dyn_hair_back_02_m}
    constraintSetup [{targetBone {boneName=head}}]
    constraintWeights [1]
   }
   n48 = "Constraint - dyn_hair_front_02_r (P)"(ComponentPose=@n47, Alpha=1, ConstraintWeights_0=1) {
    alphaBoolBlend {blendOption=Linear}
    boneToModify {boneName=dyn_hair_front_02_r}
    constraintSetup [{targetBone {boneName=head}}]
    constraintWeights [1]
   }
   n49 = "Constraint - dyn_hair_back_03_r (PP)"(ComponentPose=@n48, Alpha=1, ConstraintWeights_0=0.7, ConstraintWeights_1=0.3) {
    alphaBoolBlend {blendOption=Linear}
    boneToModify {boneName=dyn_hair_back_03_r}
    constraintSetup [{targetBone {boneName=neck_01}}, {targetBone {boneName=clavicle_r}}]
    constraintWeights [1,1]
   }
   n50 = "Constraint - dyn_hair_back_04_r (PP)"(ComponentPose=@n49, Alpha=1, ConstraintWeights_0=0.5, ConstraintWeights_1=0.5) {
    alphaBoolBlend {blendOption=Linear}
    boneToModify {boneName=dyn_hair_back_04_r}
    constraintSetup [{targetBone {boneName=spine_05}}, {targetBone {boneName=clavicle_r}}]
    constraintWeights [1,1]
   }
   n51 = "Constraint - dyn_hair_back_03_m (P)"(ComponentPose=@n50, Alpha=1, ConstraintWeights_0=1) {
    alphaBoolBlend {blendOption=Linear}
    boneToModify {boneName=dyn_hair_back_03_m}
    constraintSetup [{targetBone {boneName=neck_01}}]
    constraintWeights [1]
   }
   n52 = "Constraint - dyn_hair_back_04_m (P)"(ComponentPose=@n51, Alpha=1, ConstraintWeights_0=1) {
    alphaBoolBlend {blendOption=Linear}
    boneToModify {boneName=dyn_hair_back_04_m}
    constraintSetup [{targetBone {boneName=spine_05}}]
    constraintWeights [1]
   }
   n53 = "Constraint - dyn_hair_back_03_l (PP)"(ComponentPose=@n52, Alpha=1, ConstraintWeights_0=0.7, ConstraintWeights_1=0.3) {
    alphaBoolBlend {blendOption=Linear}
    boneToModify {boneName=dyn_hair_back_03_l}
    constraintSetup [{targetBone {boneName=neck_01}}, {targetBone {boneName=clavicle_l}}]
    constraintWeights [1,1]
   }
   n54 = "Constraint - dyn_hair_back_04_l (PP)"(ComponentPose=@n53, Alpha=1, ConstraintWeights_0=0.5, ConstraintWeights_1=0.5) {
    alphaBoolBlend {blendOption=Linear}
    boneToModify {boneName=dyn_hair_back_04_l}
    constraintSetup [{targetBone {boneName=spine_05}}, {targetBone {boneName=clavicle_l}}]
    constraintWeights [1,1]
   }
   n55 = "Transform (Modify) Bone - Bone: dyn_hair_back_02_l"(ComponentPose=@n54, Alpha=1, Translation="-0.126,-4.426,0.000", Rotation="-4.981,-8.019,0.437") {
    rotationMode=BMM_Additive
    rotationSpace=BCS_BoneSpace
    translationMode=BMM_Additive
    translationSpace=BCS_BoneSpace
    alphaBoolBlend {blendOption=Linear}
    boneToModify {boneName=dyn_hair_back_02_l}
    rotation=<rot -4.981 -8.019 0.437>
    translation=<-0.126 -4.426>
   }
   n56 = "Transform (Modify) Bone - Bone: dyn_hair_back_02_r"(ComponentPose=@n55, Alpha=1, Translation="-0.130,-4.481,-0.000", Rotation="5,-3,0") {
    rotationMode=BMM_Additive
    rotationSpace=BCS_BoneSpace
    translationMode=BMM_Additive
    translationSpace=BCS_BoneSpace
    alphaBoolBlend {blendOption=Linear}
    boneToModify {boneName=dyn_hair_back_02_r}
    rotation {pitch=5, yaw=-3}
    translation=<-0.13 -4.481>
   }
   n57 = "Transform (Modify) Bone - Bone: dyn_hair_back_02_m"(ComponentPose=@n56, Alpha=1, Translation="-0.137,-5.375,0.000", Rotation="0,-2,0") {
    rotationMode=BMM_Additive
    rotationSpace=BCS_BoneSpace
    translationMode=BMM_Additive
    translationSpace=BCS_BoneSpace
    alphaBoolBlend {blendOption=Linear}
    boneToModify {boneName=dyn_hair_back_02_m}
    rotation {yaw=-2}
    translation=<-0.137 -5.375>
   }
   n58 = "Transform (Modify) Bone - Bone: dyn_hair_back_02_l"(ComponentPose=@n57, Alpha=1, Translation="0,3,0", Rotation="0,0,0") {
    rotationMode=BMM_Additive
    rotationSpace=BCS_BoneSpace
    translationMode=BMM_Additive
    translationSpace=BCS_BoneSpace
    alphaBoolBlend {blendOption=Linear}
    boneToModify {boneName=dyn_hair_back_02_l}
    translation {y=3}
   }
   n59 = "Transform (Modify) Bone - Bone: dyn_hair_back_02_r"(ComponentPose=@n58, Alpha=1, Translation="0,3,0", Rotation="0,0,0") {
    rotationMode=BMM_Additive
    rotationSpace=BCS_BoneSpace
    translationMode=BMM_Additive
    translationSpace=BCS_BoneSpace
    alphaBoolBlend {blendOption=Linear}
    boneToModify {boneName=dyn_hair_back_02_r}
    translation {y=3}
   }
   n60 = "Transform (Modify) Bone - Bone: dyn_hair_back_02_m"(ComponentPose=@n59, Alpha=1, Translation="0,3,0", Rotation="0,0,0") {
    rotationMode=BMM_Additive
    rotationSpace=BCS_BoneSpace
    translationMode=BMM_Additive
    translationSpace=BCS_BoneSpace
    alphaBoolBlend {blendOption=Linear}
    boneToModify {boneName=dyn_hair_back_02_m}
    translation {y=3}
   }
   n61 = "Transform (Modify) Bone - Bone: dyn_hair_back_02_l"(ComponentPose=@n60, Alpha=1, Translation="0,3,0", Rotation="-0.000,-15.000,0.000") {
    rotationMode=BMM_Additive
    rotationSpace=BCS_BoneSpace
    translationMode=BMM_Additive
    translationSpace=BCS_BoneSpace
    alphaBoolBlend {blendOption=Linear}
    boneToModify {boneName=dyn_hair_back_02_l}
    rotation=<rot 0 -15 0>
    translation {y=3}
   }
   n62 = "Transform (Modify) Bone - Bone: dyn_hair_back_02_r"(ComponentPose=@n61, Alpha=1, Translation="0,3,0", Rotation="0.000,-15.000,-0.000") {
    rotationMode=BMM_Additive
    rotationSpace=BCS_BoneSpace
    translationMode=BMM_Additive
    translationSpace=BCS_BoneSpace
    alphaBoolBlend {blendOption=Linear}
    boneToModify {boneName=dyn_hair_back_02_r}
    rotation=<rot 0 -15 0>
    translation {y=3}
   }
   n63 = "Transform (Modify) Bone - Bone: dyn_hair_back_02_m"(ComponentPose=@n62, Alpha=1, Translation="0,3,0", Rotation="-0.000,-15.000,0.000") {
    rotationMode=BMM_Additive
    rotationSpace=BCS_BoneSpace
    translationMode=BMM_Additive
    translationSpace=BCS_BoneSpace
    alphaBoolBlend {blendOption=Linear}
    boneToModify {boneName=dyn_hair_back_02_m}
    rotation=<rot 0 -14.9999 0>
    translation {y=3}
   }
   n64 = "Transform (Modify) Bone - Bone: dyn_hair_back_03_l"!(ComponentPose=@n63, Alpha=1, Translation="2.369,-3.793,0.000", Rotation="-0.000,35.000,0.000") -> (Pose) {
    rotationMode=BMM_Additive
    rotationSpace=BCS_BoneSpace
    translationMode=BMM_Additive
    translationSpace=BCS_BoneSpace
    alphaBoolBlend {blendOption=Linear}
    boneToModify {boneName=dyn_hair_back_03_l}
    rotation=<rot 0 35.0001 0>
    translation=<2.369 -3.793>
   }
   n65 = "Transform (Modify) Bone - Bone: dyn_hair_back_03_r"!(ComponentPose=@n64, Alpha=1, Translation="2.775,-3.592,-0.000", Rotation="-0.000,35.000,-0.000") -> (Pose) {
    rotationMode=BMM_Additive
    rotationSpace=BCS_BoneSpace
    translationMode=BMM_Additive
    translationSpace=BCS_BoneSpace
    alphaBoolBlend {blendOption=Linear}
    boneToModify {boneName=dyn_hair_back_03_r}
    rotation=<rot 0 35 0>
    translation=<2.775 -3.592>
   }
   n66 = "Transform (Modify) Bone - Bone: dyn_hair_back_03_m"!(ComponentPose=@n65, Alpha=1, Translation="0.456,-0.549,0.000", Rotation="-0.000,15.000,-0.000") -> (Pose) {
    rotationMode=BMM_Additive
    rotationSpace=BCS_BoneSpace
    translationMode=BMM_Additive
    translationSpace=BCS_BoneSpace
    alphaBoolBlend {blendOption=Linear}
    boneToModify {boneName=dyn_hair_back_03_m}
    rotation=<rot 0 15 0>
    translation=<0.456 -0.549>
   }
   Output Pose(Result=Blend Poses by bool(BlendPose_0=Use cached pose 'RBAN'(), BlendPose_1=Component To Local(ComponentPose=@n66), bActiveValue=Get bRigidBodyEnabled()) {blendType=Linear}) {layerGroup=None}
   n67 = Get Linear Scale Backpack() -> (Linear Scale Backpack:Vector)
   n68 = RigidBody!(
    ComponentPose=Local To Component(LocalPose=Use cached pose 'Cache_BackpackState'())
    Alpha
    OverrideWorldGravity:Vector=Get GravityOverride()
    ExternalForce
    ComponentLinearAccScale:Vector=@n67
    ComponentLinearVelScale:Vector=@n67
   ) -> (Pose) {
    bClampLinearTranslationLimitToRefPose=true
    bOverrideWorldGravity=true
    overridePhysicsAsset=/Game/Characters/Player/Female/Medium/Bodies/F_MED_BoneMarrow/Meshes/Parts/F_MED_BoneMarrow_FaceAcc_Physics_Back
    simulationSpace=BaseBoneSpace
    alphaBoolBlend {blendOption=Linear}
    baseBoneRef {boneName=spine_05}
    componentAppliedLinearAccClamp=<3000 3000 3000>
    componentLinearAccScale=<0.6 0.2 0.2>
    componentLinearVelScale=<0.6 0.2 0.2>
    simSpaceSettings {worldAlpha=0.15}
   }
   n69 = Component To Local(ComponentPose=@n68)
   n70 = Get Hair Up() -> (Hair Up:double)
   n71 = "Transform (Modify) Bone - Bone: base_hair_back_l"!(ComponentPose=Local To Component(LocalPose=Use cached pose 'Cache_NormalState'()), Alpha=@n70, Translation="0,0,0") -> (Pose) {
    rotationMode=BMM_Additive
    rotationSpace=BCS_BoneSpace
    translationSpace=BCS_BoneSpace
    alphaBoolBlend {blendOption=Linear}
    alphaScaleBiasClamp {bInterpResult=true, interpSpeedDecreasing=5, interpSpeedIncreasing=7}
    boneToModify {boneName=base_hair_back_l}
    rotation {yaw=-15}
   }
   n72 = "Transform (Modify) Bone - Bone: base_hair_back_m"!(ComponentPose=@n71, Alpha=@n70, Translation="0,0,0") -> (Pose) {
    rotationMode=BMM_Additive
    rotationSpace=BCS_BoneSpace
    translationSpace=BCS_BoneSpace
    alphaBoolBlend {blendOption=Linear}
    alphaScaleBiasClamp {bInterpResult=true, interpSpeedDecreasing=5, interpSpeedIncreasing=7}
    boneToModify {boneName=base_hair_back_m}
    rotation {yaw=-20}
   }
   n73 = "Transform (Modify) Bone - Bone: base_hair_back_r"!(ComponentPose=@n72, Alpha=@n70, Translation="0,0,0") -> (Pose) {
    rotationMode=BMM_Additive
    rotationSpace=BCS_BoneSpace
    translationSpace=BCS_BoneSpace
    alphaBoolBlend {blendOption=Linear}
    alphaScaleBiasClamp {bInterpResult=true, interpSpeedDecreasing=5, interpSpeedIncreasing=7}
    boneToModify {boneName=base_hair_back_r}
    rotation {yaw=-25}
   }
   n74 = Get Linear Scale Free() -> (Linear Scale Free:Vector)
   n75 = RigidBody(ComponentPose=@n73, OverrideWorldGravity=Get GravityOverride(), ComponentLinearAccScale=@n74, ComponentLinearVelScale=@n74) {
    bClampLinearTranslationLimitToRefPose=true
    bOverrideWorldGravity=true
    overridePhysicsAsset=&F_MED_BoneMarrow_FaceAcc_Physics_Back
    simulationSpace=BaseBoneSpace
    alphaBoolBlend {blendOption=Linear}
    baseBoneRef {boneName=spine_05}
    componentAppliedLinearAccClamp=<3000 3000 3000>
    componentLinearAccScale=<0.6 0.2 0.2>
    componentLinearVelScale=<0.6 0.2 0.2>
    simSpaceSettings {worldAlpha=0.125}
   }
   n76 = Component To Local(ComponentPose=@n75)
   n77 = Get isWearingBackbling()
   Save cached pose 'RBAN'(Pose=Blend Poses by bool(
     BlendPose_0=Use cached pose 'Cache_BackpackState'()
     BlendPose_1=Blend Poses by bool(BlendPose_0=@n69, BlendPose_1=@n76, bActiveValue=@n77) {blendType=Linear}
     bActiveValue=Get Is Wearing Cape() -> (Is Wearing Cape:bool)
    ) {blendType=Linear}) {cacheName=RBAN}
   n78 = Control Rig!(Source=Use cached pose 'Cache_NormalState'(), Alpha, parachuting_alpha:double=Get isParachuting_alpha() -> (isParachuting_alpha:double)) -> (Pose) {
    alphaBoolBlend {blendOption=Linear}
    controlRigAssetReference {blueprintRigClass=/Game/Characters/Player/Female/Medium/Bodies/F_MED_BoneMarrow/Meshes/Parts/F_MED_BoneMarrow_FaceAcc_CtrlRig.F_MED_BoneMarrow_FaceAcc_CtrlRig_C}
   }
   Save cached pose 'Cache_BackpackState'(Pose=Component To Local(ComponentPose=Local To Component(LocalPose=@n78))) {cacheName=Cache_BackpackState}
   n79 = Layered blend per bone(
    BasePose=Copy Pose From Mesh() {bCopyCurves=true, bUseAttachedParent=true}
    BlendPoses_0=Copy Pose From Mesh(SourceMeshComponent=Get Body_Mesh())
   ) {layerSetup [{branchFilters [
       {boneName=clavicle_l}
       {boneName=clavicle_r}
       {boneName=upperarm_l}
       {boneName=upperarm_r}
       {boneName=upperarm_twist_01_l}
       {boneName=upperarm_twist_01_r}
       {boneName=deform_upperarm_bk_base_l}
       {boneName=deform_upperarm_bk_l}
       {boneName=deform_upperarm_fr_base_l}
       {boneName=deform_upperarm_fr_l}
       {boneName=deform_upperarm_up_base_l}
       {boneName=deform_upperarm_up_l}
       {boneName=deform_upperarm_bk_base_r}
       {boneName=deform_upperarm_bk_r}
       {boneName=deform_upperarm_fr_base_r}
       {boneName=deform_upperarm_fr_r}
       {boneName=deform_upperarm_up_base_r}
       {boneName=deform_upperarm_up_r}
      ]}]}
   Save cached pose 'Cache_StartPose'(Pose=@n79) {cacheName=Cache_StartPose}
   n80 = RigidBody(
    ComponentPose=Local To Component(LocalPose=Use cached pose 'Cache_StartPose'())
    OverrideWorldGravity=Get GravityOverride()
    ComponentLinearAccScale="0.1,0.1,0.1"
    ComponentLinearVelScale="0.1,0.1,0.1"
   ) {
    bClampLinearTranslationLimitToRefPose=true
    bOverrideWorldGravity=true
    overridePhysicsAsset=/Game/Characters/Player/Female/Medium/Bodies/F_MED_BoneMarrow/Meshes/Parts/F_MED_BoneMarrow_FaceAcc_SkyDiving_Physics
    simulationSpace=BaseBoneSpace
    alphaBoolBlend {blendOption=Linear}
    baseBoneRef {boneName=head}
    componentAppliedLinearAccClamp=<3000 3000 3000>
    componentLinearAccScale=<0.6 0.2 0.2>
    componentLinearVelScale=<0.6 0.2 0.2>
    overrideWorldGravity {z=-980}
    simSpaceSettings {velocityScaleZ=0.2}
   }
   n81 = Component To Local(ComponentPose=@n80)
   n82 = RigidBody(
    ComponentPose=Local To Component(LocalPose=Use cached pose 'Cache_StartPose'())
    OverrideWorldGravity=Get GravityOverride()
    ComponentLinearAccScale="0.35,0.3,0.3"
    ComponentLinearVelScale="0.35,0.3,0.3"
   ) {
    bClampLinearTranslationLimitToRefPose=true
    bOverrideWorldGravity=true
    overridePhysicsAsset=/Game/Characters/Player/Female/Medium/Bodies/F_MED_BoneMarrow/Meshes/Parts/F_MED_BoneMarrow_FaceAcc_Physics
    simulationSpace=BaseBoneSpace
    alphaBoolBlend {blendOption=Linear}
    baseBoneRef {boneName=spine_05}
    componentAppliedLinearAccClamp=<3000 3000 3000>
    componentLinearAccScale=<0.6 0.2 0.2>
    componentLinearVelScale=<0.6 0.2 0.2>
    overrideWorldGravity {z=-980}
    simSpaceSettings {worldAlpha=0.15}
   }
   n83 = Component To Local(ComponentPose=@n82)
   n84 = Get playemoteskydivingisinfrontend() -> (playemoteskydivingisinfrontend:bool)
   n85 = Anim Dynamics - Bone: base_hair_frontshort_r(
    ComponentPose=Local To Component(LocalPose=Blend Poses by bool(BlendPose_0=@n81, BlendPose_1=@n83, BlendTime_0=0.4, bActiveValue=@n84))
    Alpha=1
    ExternalForce="0,0,0"
   ) {
    angularSpringConstant=1
    bAngularSpring=true
    bUseGravityOverride=true
    alphaBoolBlend {blendOption=Linear}
    boundBone {boneName=base_hair_frontshort_r}
    gravityOverride {z=-980}
    physicsBodyDefinitions [{
      boundBone {boneName=base_hair_frontshort_r}
      boxExtents=<10 10 10>
      constraintSetup {angularLimitsMax=<0 0 15>, angularLimitsMin=<0 0 0>, angularTarget=<0 0 0>}
      localJointOffset=<1 0 0>
     }]
   }
   n86 = Anim Dynamics - Bone: base_hair_frontshort_m(ComponentPose=@n85, Alpha=1, ExternalForce="0,0,0") {
    angularSpringConstant=1
    bAngularSpring=true
    bUseGravityOverride=true
    alphaBoolBlend {blendOption=Linear}
    boundBone {boneName=base_hair_frontshort_m}
    gravityOverride {z=-980}
    physicsBodyDefinitions [{
      boundBone {boneName=base_hair_frontshort_m}
      boxExtents=<10 10 10>
      constraintSetup {angularLimitsMax=<0 0 15>, angularLimitsMin=<0 0 0>, angularTarget=<0 0 0>}
      localJointOffset=<0 0 0>
     }]
   }
   n87 = Anim Dynamics - Bone: base_hair_frontshort_l(ComponentPose=@n86, Alpha=1, ExternalForce="0,0,0") {
    angularSpringConstant=1
    bAngularSpring=true
    bUseGravityOverride=true
    alphaBoolBlend {blendOption=Linear}
    boundBone {boneName=base_hair_frontshort_l}
    gravityOverride {z=-980}
    physicsBodyDefinitions [{
      boundBone {boneName=base_hair_frontshort_l}
      boxExtents=<10 10 10>
      constraintSetup {angularLimitsMax=<0 0 15>, angularLimitsMin=<0 0 0>, angularTarget=<0 0 0>}
      localJointOffset=<0 0 0>
     }]
   }
   Save cached pose 'Cache_NormalState'(Pose=Component To Local(ComponentPose=@n87)) {cacheName=Cache_NormalState}
   Get FE_Dyn_Alpha() -> (FE_Dyn_Alpha:double)
   Get FE_Dyn_Alpha()
  ]}
 EventGraph {nodes [
   Event OnCharacterCustomizationLoaded()
   > Sequence()
   > n1 = Cast To PlayerPawn_Athena(Object=TryGetPawnOwner())
   > Set Body_Mesh(Body_Mesh=GetSkeletalMeshForPartType(self=@n1))
   Event BlueprintInitializeAnimation()
   > Set Skel Mesh(Skel Mesh=GetOwnerSkeletalMeshComponent())
   Event BlueprintUpdateAnimation()
   > n2 = Sequence!(execute) -> (then_0, then_1, then_2:Exec, then_3:Exec, then_4:Exec, then_5:Exec, then_6:Exec)
   n3 = GetAlphaFromReferenceTransformDelta(SkeletalMeshComponent=Get Skel Mesh(), BoneName=spine_05, BaseName=head, InputRange=(X=0,Y=70), TwistAxis="0,0,1")
   Set Hair Up(execute=@n2.then_2, Hair Up:double=@n3) -> (Output_Get:double)
   > Set IsTargetingHeadUp(IsTargetingHeadUp:double=float * float(A=@n3, B=Get IsTargeting() -> (IsTargeting:double)) {is=title K2Node_PromotableOperator}) -> (Output_Get:double)
   > Set playemoteskydivingisinfrontend(playemoteskydivingisinfrontend:bool=OR Boolean!(A=Get bIsPlayingEmote(), B=Get bIsSkyDiving(), C:bool=Get bIsInFrontEnd()) -> (ReturnValue)) -> (Output_Get:bool)
   n4 = Gravity Override State Function DEPRECATED(
    execute=@n2.then_0
    CharacterPartAnimInstance=Self-Reference()
    GameTimeInSeconds=GetGameTimeInSeconds()
    GravityOverrideParams=Get Gravity Override Params()
   )
   > Set GravityOverride(GravityOverride=@n4)
   n5 = CheckForBackbling()
   > Set isWearingBackbling(isWearingBackbling=AND Boolean!(A=@n5, B=NOT Boolean(A=@n5.bIsBackblingACape), C:bool=NOT Boolean(A=@n5.bIsBackblingLong)) -> (ReturnValue))
   > Set Is Wearing Cape(Is Wearing Cape=OR Boolean!(A=@n5.bIsBackblingACape, B=@n5.bIsBackblingLong) -> (ReturnValue))
   n6 = Get bRigidBodyEnabled()
   > Set RBANoff(RBANoff:double=SelectFloat(A=1, B=0, bPickA=NOT Boolean(A=@n6))) -> (Output_Get:double)
   n7 = Get bIsInFrontEnd()
   n8 = Get bIsSkyDiving()
   > Set FE_Skydive(FE_Skydive:double=SelectFloat(A=1, B=0, bPickA=AND Boolean!(A=@n8, B=@n7) -> (ReturnValue))) -> (Output_Get:double)
   > n9 = Branch(Condition=Get bIsInFrontEnd())
   > Set FE_Dyn_Alpha(FE_Dyn_Alpha:double=0.4) -> (Output_Get:double)
   Set FE_Dyn_Alpha(execute=@n9.else, FE_Dyn_Alpha=1)
   Set IsTargeting(execute=@n2.then_1, IsTargeting:double="To Float (Boolean)"(InBool=Get bIsTargeting())) -> (Output_Get:double)
   n10 = Get bIsSkydiveDiveMode()
   n11 = Get bIsSkydiveFloating()
   > Set isSkyDiving(isSkyDiving:float="To Float (Boolean)"(InBool=OR Boolean(A=@n10, B=@n11))) -> (Output_Get:float)
   > Set isParachuting_alpha(isParachuting_alpha:double="To Float (Boolean)"(InBool=Get bIsParachuteOpened())) -> (Output_Get:double)
   > n12 = Branch(Condition=Get bIsSkyDiving())
   > Set Linear Scale Free(Linear Scale Free:Vector="0.3,0.3,0.3") -> (Output_Get:Vector)
   > Set Linear Scale Backpack(Linear Scale Backpack:Vector="0.3,0.2,0.2") -> (Output_Get:Vector)
   > Set Linear Scale Cape(Linear Scale Cape:Vector="0.25,0.2,0.2") -> (Output_Get:Vector)
   > Set Linear Scale Free_Front(Linear Scale Free_Front:Vector="0.1,0.1,0.1") -> (Output_Get:Vector)
   Set Linear Scale Free(execute=@n12.else, Linear Scale Free="0.23,0.26,0.26")
   > Set Linear Scale Backpack(Linear Scale Backpack="0.3,0.2,0.2")
   > Set Linear Scale Cape(Linear Scale Cape="0.25,0.2,0.2")
   > Set Linear Scale Free_Front(Linear Scale Free_Front="0.35,0.3,0.3")
  ]}
]
variables [
 Body_Mesh {type=SkeletalMeshComponent}
 FE_Dyn_Alpha {type=number}
 FE_Skydive {type=number}
 Gravity Override Params {type=GravityOverrideParamsStruct, default {skydive_Parachuting {windAmplitude=<100 -250 200>, windFrequency=<1.5 4 3>}}}
 GravityOverride {type=Vector}
 Hair Up {type=number}
 Is Wearing Cape {type=boolean}
 isParachuting_alpha {type=number}
 isSkyDiving {type=number}
 IsTargeting {type=number}
 IsTargetingHeadUp {type=number}
 isWearingBackbling {type=boolean}
 Linear Scale Backpack {type=Vector}
 Linear Scale Cape {type=Vector}
 Linear Scale Free {type=Vector}
 Linear Scale Free_Front {type=Vector}
 playemoteskydivingisinfrontend {type=boolean}
 RBANoff {type=number}
 Skel Mesh {type=SkeletalMeshComponent}
 skydiving_frontHair {type=number}
]
